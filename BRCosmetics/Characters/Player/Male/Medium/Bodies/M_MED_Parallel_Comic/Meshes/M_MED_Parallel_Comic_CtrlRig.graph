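# Control Rig for the M_MED_Parallel_Comic player body that procedurally drives twist and corrective deform bones.
# Swing-twist extraction from hands, feet and thighs drives forearm and calf twists, plus inverted glute and thigh twists.
# Clamped joint-angle remaps blend elbow, knee and front-thigh correctives, while actor_emote switches the pelvis offset and knee shapes.
# FK-control loops and back-thigh correctives are disconnected from execution, and both is_arms_crossed branches write zero offsets.
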ControlRigBlueprint /BRCosmetics/Characters/Player/Male/Medium/Bodies/M_MED_Parallel_Comic/Meshes/M_MED_Parallel_Comic_CtrlRig
graphs [
 RigVMFunctionLibrary {links [], nodes []}
 RigVMModel {nodes [
   n1 = &RigUnit_BeginExecution()
   n2 = /Script/ControlRig.RigUnit_SequenceExecution(ExecuteContext=@n1.ExecutePin) {is=struct RigVMUnitNode}
   n3 = &RigUnit_SequenceExecution(ExecuteContext=@n2.B)
   n4 = &RigUnit_SequenceExecution(ExecuteContext=@n3.A)
   n5 = &RigVMFunction_MathQuaternionSwingTwist(
    TwistAxis=(X=1,Y=0,Z=0)
    Input=&RigVMFunction_MathTransformMakeRelative(
     Parent=&RigUnit_GetTransform(Item=(Type=Bone,Name="hand_l"), Space=LocalSpace, bInitial=true).Transform
     Global=&RigUnit_GetTransform(Item=(Type=Bone,Name="hand_l"), Space=LocalSpace).Transform
    ).Local.Rotation
   )
   &RigUnit_SetRotation(Item=(Type=Bone,Name="lowerarm_twist_01_l"), Weight=1, bPropagateToChildren=false, Value=@n5.Twist, ExecutePin=@n4.B)
   > &RigUnit_SetRotation(Item=(Type=Bone,Name="lowerarm_twist_02_l"), Weight=0.5, bPropagateToChildren=false, Value=@n5.Twist)
   n6 = &RigVMFunction_MathQuaternionSwingTwist(
    TwistAxis=(X=1,Y=0,Z=0)
    Input=&RigVMFunction_MathTransformMakeRelative(
     Parent=&RigUnit_GetTransform(Item=(Type=Bone,Name="hand_r"), Space=LocalSpace, bInitial=true).Transform
     Global=&RigUnit_GetTransform(Item=(Type=Bone,Name="hand_r"), Space=LocalSpace).Transform
    ).Local.Rotation
   )
   > &RigUnit_SetRotation(Item=(Type=Bone,Name="lowerarm_twist_01_r"), Weight=1, bPropagateToChildren=false, Value=@n6.Twist)
   > &RigUnit_SetRotation(Item=(Type=Bone,Name="lowerarm_twist_02_r"), Weight=0.5, bPropagateToChildren=false, Value=@n6.Twist)
   Branch_1_RigVMUnitNode = &RigVMFunction_ControlFlowBranch(Condition=RigVMVariableNode() {variable=actor_emote}.Value, ExecuteContext=@n4.A)
   RigVMVariableNode(ExecuteContext=@Branch_1_RigVMUnitNode.False) {variable=emote_glut_weight}
   > RigVMVariableNode() {variable=emote_thigh_weight}
   > &RigUnit_SetTransform(
    Item=(Type=Bone,Name="deform_pelvis")
    Space=LocalSpace
    Value=(Rotation=(X=0,Y=0,Z=0,W=1),Translation=(X=0,Y=0,Z=0),Scale3D=(X=1,Y=1,Z=1))
    Weight=1
    bInitial=false
   )
   RigVMVariableNode(ExecuteContext=@Branch_1_RigVMUnitNode.True) {variable=emote_glut_weight}
   > RigVMVariableNode() {variable=emote_thigh_weight}
   > &RigUnit_SetTransform(
    Item=(Type=Bone,Name="deform_pelvis")
    Space=LocalSpace
    Value=(Rotation=(X=0,Y=0,Z=0,W=1),Translation=(X=-6,Y=0,Z=0),Scale3D=(X=1,Y=1,Z=1))
    Weight=1
    bInitial=false
   )
   n7 = &RigUnit_SequenceExecution(ExecuteContext=@n4.C)
   Branch_2_RigVMUnitNode = &RigVMFunction_ControlFlowBranch(ExecuteContext=@n7.A, Condition=RigVMVariableNode() {variable=is_arms_crossed}.Value)
   RigVMVariableNode(Value=(X=0,Y=0,Z=0), ExecuteContext=@Branch_2_RigVMUnitNode.False) {variable=glut_twist_arms_crossed_l}
   > RigVMVariableNode(Value=(X=0,Y=0,Z=0)) {variable=thigh_twist_arms_crossed_l}
   > RigVMVariableNode(Value=(X=0,Y=0,Z=0)) {variable=glut_twist_arms_crossed_r}
   > RigVMVariableNode(Value=(X=0,Y=0,Z=0)) {variable=thigh_twist_arms_crossed_r}
   RigVMVariableNode(Value=(X=0,Y=0,Z=0), ExecuteContext=@Branch_2_RigVMUnitNode.True) {variable=glut_twist_arms_crossed_l}
   > RigVMVariableNode(Value=(X=0,Y=0,Z=0)) {variable=thigh_twist_arms_crossed_l}
   > RigVMVariableNode(Value=(X=0,Y=0,Z=0)) {variable=glut_twist_arms_crossed_r}
   > RigVMVariableNode(Value=(X=0,Y=0,Z=0)) {variable=thigh_twist_arms_crossed_r}
   n8 = &RigVMFunction_MathQuaternionSwingTwist(
    TwistAxis=(X=1,Y=0,Z=0)
    Input=&RigVMFunction_MathTransformMakeRelative(
     Parent=&RigUnit_GetTransform(Item=(Type=Bone,Name="thigh_l"), Space=LocalSpace, bInitial=true).Transform
     Global=&RigUnit_GetTransform(Item=(Type=Bone,Name="thigh_l"), Space=LocalSpace).Transform
    ).Local.Rotation
   )
   n9 = /Script/RigVM.RigVMFunction_MathVectorAdd(
    A=&RigVMFunction_MathQuaternionToEuler(RotationOrder=ZYX, Value=@n8.Twist).Result
    B=RigVMVariableNode() {variable=glut_twist_arms_crossed_l}.Value
   ) {is=struct RigVMUnitNode, template="Add::Execute(in A,in B,out Result)"}
   &RigUnit_SetRotation(
    Item=(Type=Bone,Name="deform_glut_l")
    bPropagateToChildren=false
    Value=/Script/RigVM.RigVMFunction_MathQuaternionInverse(Value=&RigVMFunction_MathQuaternionFromEuler(Euler=@n9.Result).Result) {is=struct RigVMUnitNode, template="Inverse::Execute(in Value,out Result)"}.Result
    ExecutePin=@n7.C
    Weight=RigVMVariableNode() {variable=emote_glut_weight}.Value
   )
   n10 = &RigVMFunction_MathVectorAdd(
    A=&RigVMFunction_MathQuaternionToEuler(RotationOrder=ZYX, Value=@n8.Twist).Result
    B=RigVMVariableNode() {variable=thigh_twist_arms_crossed_l}.Value
   ) {template="Add::Execute(in A,in B,out Result)"}
   > &RigUnit_SetRotation(
    Item=(Type=Bone,Name="thigh_twist_01_l")
    Value=&RigVMFunction_MathQuaternionInverse(Value=&RigVMFunction_MathQuaternionFromEuler(Euler=@n10.Result).Result) {template="Inverse::Execute(in Value,out Result)"}.Result
    Weight=RigVMVariableNode() {variable=emote_thigh_weight}.Value
   )
   n11 = &RigVMFunction_MathQuaternionSwingTwist(
    TwistAxis=(X=1,Y=0,Z=0)
    Input=&RigVMFunction_MathTransformMakeRelative(
     Parent=&RigUnit_GetTransform(Item=(Type=Bone,Name="thigh_r"), Space=LocalSpace, bInitial=true).Transform
     Global=&RigUnit_GetTransform(Item=(Type=Bone,Name="thigh_r"), Space=LocalSpace).Transform
    ).Local.Rotation
   )
   n12 = &RigVMFunction_MathVectorAdd(
    A=&RigVMFunction_MathQuaternionToEuler(RotationOrder=ZYX, Value=@n11.Twist).Result
    B=RigVMVariableNode() {variable=glut_twist_arms_crossed_r}.Value
   ) {template="Add::Execute(in A,in B,out Result)"}
   &RigUnit_SetRotation(
    Item=(Type=Bone,Name="deform_glut_r")
    bPropagateToChildren=false
    Value=&RigVMFunction_MathQuaternionInverse(Value=&RigVMFunction_MathQuaternionFromEuler(Euler=@n12.Result).Result) {template="Inverse::Execute(in Value,out Result)"}.Result
    Weight=RigVMVariableNode() {variable=emote_glut_weight}.Value
    ExecutePin=@n7.B
   )
   n13 = &RigVMFunction_MathVectorAdd(
    A=&RigVMFunction_MathQuaternionToEuler(RotationOrder=ZYX, Value=@n11.Twist).Result
    B=RigVMVariableNode() {variable=thigh_twist_arms_crossed_r}.Value
   ) {template="Add::Execute(in A,in B,out Result)"}
   > &RigUnit_SetRotation(
    Item=(Type=Bone,Name="thigh_twist_01_r")
    Value=&RigVMFunction_MathQuaternionInverse(Value=&RigVMFunction_MathQuaternionFromEuler(Euler=@n13.Result).Result) {template="Inverse::Execute(in Value,out Result)"}.Result
    Weight=RigVMVariableNode() {variable=emote_thigh_weight}.Value
   )
   n14 = &RigVMFunction_MathQuaternionSwingTwist(
    TwistAxis=(X=1,Y=0,Z=0)
    Input=&RigVMFunction_MathTransformMakeRelative(
     Parent=&RigUnit_GetTransform(Item=(Type=Bone,Name="foot_l"), Space=LocalSpace, bInitial=true).Transform
     Global=&RigUnit_GetTransform(Item=(Type=Bone,Name="foot_l"), Space=LocalSpace).Transform
    ).Local.Rotation
   )
   &RigUnit_SetRotation(Item=(Type=Bone,Name="calf_twist_01_l"), Weight=0.9, bPropagateToChildren=false, Value=@n14.Twist, ExecutePin=@n4.D)
   > &RigUnit_SetRotation(Item=(Type=Bone,Name="calf_twist_02_l"), Weight=0.5, bPropagateToChildren=false, Value=@n14.Twist)
   n15 = &RigVMFunction_MathQuaternionSwingTwist(
    TwistAxis=(X=1,Y=0,Z=0)
    Input=&RigVMFunction_MathTransformMakeRelative(
     Parent=&RigUnit_GetTransform(Item=(Type=Bone,Name="foot_r"), Space=LocalSpace, bInitial=true).Transform
     Global=&RigUnit_GetTransform(Item=(Type=Bone,Name="foot_r"), Space=LocalSpace).Transform
    ).Local.Rotation
   )
   > &RigUnit_SetRotation(Item=(Type=Bone,Name="calf_twist_01_r"), Weight=0.9, bPropagateToChildren=false, Value=@n15.Twist)
   > &RigUnit_SetRotation(Item=(Type=Bone,Name="calf_twist_02_r"), Weight=0.5, bPropagateToChildren=false, Value=@n15.Twist)
   n16 = &RigUnit_SequenceExecution(ExecuteContext=@n3.B)
   n17 = &RigUnit_SequenceExecution(ExecuteContext=@n16.C)
   n18 = &RigUnit_SequenceExecution(ExecuteContext=@n17.C)
   n19 = &RigUnit_GetRelativeBoneTransform(Bone=thigh_r, Space=pelvis)
   n20 = &RigVMFunction_MathFloatRemap(
    SourceMaximum=77
    SourceMinimum=177
    TargetMinimum=0
    bClamp=true
    Value=/Script/RigVM.RigVMFunction_MathFloatAbs(Value=&RigVMFunction_MathQuaternionToEuler(RotationOrder=ZYX, Value=@n19.Transform.Rotation).Result.Z) {is=struct RigVMUnitNode, template="Absolute::Execute(in Value,out Result)"}.Result
   )
   n21 = &RigVMFunction_MathQuaternionToEuler(
    RotationOrder=ZYX
    Value=/Script/ControlRig.RigUnit_GetInitialBoneTransform(Bone=deform_thigh_fr_r, Space=LocalSpace) {is=struct RigVMUnitNode}.Transform.Rotation
   )
   /Script/ControlRig.RigUnit_ModifyBoneTransforms(
    BoneToModify=((Bone="deform_thigh_fr_r",Transform=(Rotation=(X=0,Y=0,Z=0,W=1),Translation=(X=-13.423,Y=-8.962,Z=-2.253),Scale3D=(X=1,Y=1,Z=1.175))))
    Mode=OverrideLocal
    WeightMaximum=1
    Weight=@n20.Result
    BoneToModify.0.Transform.Rotation=&RigVMFunction_MathQuaternionFromEuler(Euler=&RigVMFunction_MathVectorAdd(B=(X=0,Y=-14.949,Z=0), A=@n21.Result) {template="Add::Execute(in A,in B,out Result)"}.Result).Result
    ExecutePin=@n18.A
   ) {is=struct RigVMUnitNode}
   n22 = &RigUnit_GetRelativeBoneTransform(Bone=thigh_l, Space=pelvis)
   n23 = &RigVMFunction_MathFloatRemap(
    SourceMaximum=100
    SourceMinimum=2
    TargetMinimum=0
    bClamp=true
    Value=&RigVMFunction_MathFloatAbs(Value=&RigVMFunction_MathQuaternionToEuler(RotationOrder=ZYX, Value=@n22.Transform.Rotation).Result.Z) {template="Absolute::Execute(in Value,out Result)"}.Result
   )
   n24 = &RigVMFunction_MathQuaternionToEuler(RotationOrder=ZYX, Value=&RigUnit_GetInitialBoneTransform(Bone=deform_thigh_fr_l, Space=LocalSpace).Transform.Rotation)
   &RigUnit_ModifyBoneTransforms(
    BoneToModify=((Bone="deform_thigh_fr_l",Transform=(Rotation=(X=0,Y=0,Z=0,W=1),Translation=(X=13.423,Y=8.962,Z=2.253),Scale3D=(X=1,Y=1,Z=1.175))))
    Mode=OverrideLocal
    WeightMaximum=1
    Weight=@n23.Result
    BoneToModify.0.Transform.Rotation=&RigVMFunction_MathQuaternionFromEuler(Euler=&RigVMFunction_MathVectorAdd(B=(X=0,Y=-14.949,Z=0), A=@n24.Result) {template="Add::Execute(in A,in B,out Result)"}.Result).Result
    ExecutePin=@n18.B
   )
   Branch_RigVMUnitNode = &RigVMFunction_ControlFlowBranch(Condition=RigVMVariableNode() {variable=actor_emote}.Value, ExecuteContext=@n17.B)
   n25 = &RigUnit_SequenceExecution(ExecuteContext=@Branch_RigVMUnitNode.False)
   n26 = &RigVMFunction_MathQuaternionToEuler(
    RotationOrder=ZYX
    Value=&RigUnit_GetRelativeTransformForItem(Child=(Type=Bone,Name="calf_l"), Parent=(Type=Bone,Name="thigh_l"), bChildInitial=false, bParentInitial=false).RelativeTransform.Rotation
   )
   n27 = &RigVMFunction_MathFloatRemap(SourceMaximum=90, SourceMinimum=0, TargetMinimum=0, bClamp=true, Value=@n26.Result.Z)
   n28 = &RigVMFunction_MathQuaternionToEuler(
    RotationOrder=ZYX
    Value=&RigUnit_GetTransform(Item=(Type=Bone,Name="deform_knee_bk_01_l"), Space=LocalSpace, bInitial=true).Transform.Rotation
   )
   n29 = &RigVMFunction_MathQuaternionFromEuler(Euler=&RigVMFunction_MathVectorAdd(B=(X=0,Y=0,Z=25), A=@n28.Result) {template="Add::Execute(in A,in B,out Result)"}.Result)
   &RigUnit_ModifyTransforms(
    ItemToModify=((Item=(Type=Bone,Name="deform_knee_bk_01_l"),Transform=(Rotation=(X=0,Y=0,Z=0,W=1),Translation=(X=-3,Y=-6,Z=0),Scale3D=(X=1,Y=1,Z=1))))
    Mode=OverrideLocal
    Weight=@n27.Result
    ItemToModify.0.Transform.Rotation=@n29.Result
    ExecutePin=@n25.B
   )
   n30 = &RigVMFunction_MathFloatRemap(SourceMaximum=105, SourceMinimum=80, TargetMinimum=0, bClamp=true, Value=@n26.Result.Z)
   n31 = &RigVMFunction_MathQuaternionFromEuler(Euler=&RigVMFunction_MathVectorAdd(B=(X=0,Y=0,Z=-20), A=@n28.Result) {template="Add::Execute(in A,in B,out Result)"}.Result)
   > &RigUnit_ModifyTransforms(
    ItemToModify=((Item=(Type=Bone,Name="deform_knee_bk_01_l"),Transform=(Rotation=(X=0,Y=0,Z=0,W=1),Translation=(X=-7.5,Y=-6,Z=0),Scale3D=(X=1,Y=1,Z=1))))
    Mode=OverrideLocal
    Weight=@n30.Result
    ItemToModify.0.Transform.Rotation=@n31.Result
   )
   n32 = &RigVMFunction_MathQuaternionToEuler(
    RotationOrder=ZYX
    Value=&RigUnit_GetTransform(Item=(Type=Bone,Name="deform_knee_fr_01_l"), Space=LocalSpace, bInitial=true).Transform.Rotation
   )
   n33 = &RigVMFunction_MathQuaternionFromEuler(Euler=&RigVMFunction_MathVectorAdd(B=(X=0,Y=0,Z=-40), A=@n32.Result) {template="Add::Execute(in A,in B,out Result)"}.Result)
   > &RigUnit_ModifyTransforms(
    ItemToModify=((Item=(Type=Bone,Name="deform_knee_fr_01_l"),Transform=(Rotation=(X=0,Y=0,Z=0,W=1),Translation=(X=2,Y=1,Z=0),Scale3D=(X=1,Y=1,Z=1))))
    Mode=OverrideLocal
    Weight=@n27.Result
    ItemToModify.0.Transform.Rotation=@n33.Result
   )
   n34 = &RigVMFunction_MathQuaternionFromEuler(Euler=&RigVMFunction_MathVectorAdd(B=(X=0,Y=0,Z=-20), A=@n32.Result) {template="Add::Execute(in A,in B,out Result)"}.Result)
   > &RigUnit_ModifyTransforms(
    ItemToModify=((Item=(Type=Bone,Name="deform_knee_fr_01_l"),Transform=(Rotation=(X=0,Y=0,Z=0,W=1),Translation=(X=2,Y=1,Z=-2),Scale3D=(X=1.2,Y=1,Z=1.1))))
    Mode=OverrideLocal
    Weight=@n30.Result
    ItemToModify.0.Transform.Rotation=@n34.Result
   )
   n35 = &RigVMFunction_MathQuaternionToEuler(
    RotationOrder=ZYX
    Value=&RigUnit_GetRelativeTransformForItem(Child=(Type=Bone,Name="calf_r"), Parent=(Type=Bone,Name="thigh_r"), bChildInitial=false, bParentInitial=false).RelativeTransform.Rotation
   )
   n36 = &RigVMFunction_MathFloatRemap(SourceMaximum=90, SourceMinimum=0, TargetMinimum=0, bClamp=true, Value=@n35.Result.Z)
   n37 = &RigVMFunction_MathQuaternionToEuler(
    RotationOrder=ZYX
    Value=&RigUnit_GetTransform(Item=(Type=Bone,Name="deform_knee_bk_01_r"), Space=LocalSpace, bInitial=true).Transform.Rotation
   )
   n38 = &RigVMFunction_MathQuaternionFromEuler(Euler=&RigVMFunction_MathVectorAdd(B=(X=0,Y=0,Z=25), A=@n37.Result) {template="Add::Execute(in A,in B,out Result)"}.Result)
   &RigUnit_ModifyTransforms(
    ItemToModify=((Item=(Type=Bone,Name="deform_knee_bk_01_r"),Transform=(Rotation=(X=0,Y=0,Z=0,W=1),Translation=(X=3,Y=6,Z=0),Scale3D=(X=1,Y=1,Z=1))))
    Mode=OverrideLocal
    Weight=@n36.Result
    ItemToModify.0.Transform.Rotation=@n38.Result
    ExecutePin=@n25.A
   )
   n39 = &RigVMFunction_MathFloatRemap(SourceMaximum=105, SourceMinimum=80, TargetMinimum=0, bClamp=true, Value=@n35.Result.Z)
   n40 = &RigVMFunction_MathQuaternionFromEuler(Euler=&RigVMFunction_MathVectorAdd(B=(X=0,Y=0,Z=-20), A=@n37.Result) {template="Add::Execute(in A,in B,out Result)"}.Result)
   > &RigUnit_ModifyTransforms(
    ItemToModify=((Item=(Type=Bone,Name="deform_knee_bk_01_r"),Transform=(Rotation=(X=0,Y=0,Z=0,W=1),Translation=(X=7.5,Y=6,Z=0),Scale3D=(X=1,Y=1,Z=1))))
    Mode=OverrideLocal
    Weight=@n39.Result
    ItemToModify.0.Transform.Rotation=@n40.Result
   )
   n41 = &RigVMFunction_MathQuaternionToEuler(
    RotationOrder=ZYX
    Value=&RigUnit_GetTransform(Item=(Type=Bone,Name="deform_knee_fr_01_r"), Space=LocalSpace, bInitial=true).Transform.Rotation
   )
   n42 = &RigVMFunction_MathQuaternionFromEuler(Euler=&RigVMFunction_MathVectorAdd(B=(X=0,Y=0,Z=-40), A=@n41.Result) {template="Add::Execute(in A,in B,out Result)"}.Result)
   > &RigUnit_ModifyTransforms(
    ItemToModify=((Item=(Type=Bone,Name="deform_knee_fr_01_r"),Transform=(Rotation=(X=0,Y=0,Z=0,W=1),Translation=(X=-2,Y=-1,Z=0),Scale3D=(X=1,Y=1,Z=1))))
    Mode=OverrideLocal
    Weight=@n36.Result
    ItemToModify.0.Transform.Rotation=@n42.Result
   )
   n43 = &RigVMFunction_MathQuaternionFromEuler(Euler=&RigVMFunction_MathVectorAdd(B=(X=10,Y=0,Z=-20), A=@n41.Result) {template="Add::Execute(in A,in B,out Result)"}.Result)
   > &RigUnit_ModifyTransforms(
    ItemToModify=((Item=(Type=Bone,Name="deform_knee_fr_01_r"),Transform=(Rotation=(X=0,Y=0,Z=0,W=1),Translation=(X=-2,Y=-1,Z=2),Scale3D=(X=1.2,Y=1,Z=1.1))))
    Mode=OverrideLocal
    Weight=@n39.Result
    ItemToModify.0.Transform.Rotation=@n43.Result
   )
   n44 = &RigUnit_SequenceExecution(ExecuteContext=@Branch_RigVMUnitNode.True)
   &RigUnit_ModifyTransforms(
    ItemToModify=((Item=(Type=Bone,Name="deform_knee_bk_01_l"),Transform=(Rotation=(X=0,Y=0,Z=0,W=1),Translation=(X=-3,Y=-6,Z=0),Scale3D=(X=1,Y=1,Z=0.75))))
    Mode=OverrideLocal
    Weight=@n27.Result
    ItemToModify.0.Transform.Rotation=@n31.Result
    ExecutePin=@n44.B
   )
   > &RigUnit_ModifyTransforms(
    ItemToModify=((Item=(Type=Bone,Name="deform_knee_bk_01_l"),Transform=(Rotation=(X=0,Y=0,Z=0,W=1),Translation=(X=-6,Y=-3.5,Z=0),Scale3D=(X=1,Y=1,Z=0.75))))
    Mode=OverrideLocal
    Weight=@n30.Result
    ItemToModify.0.Transform.Rotation=@n29.Result
   )
   > &RigUnit_ModifyTransforms(
    ItemToModify=((Item=(Type=Bone,Name="deform_knee_fr_01_l"),Transform=(Rotation=(X=0,Y=0,Z=0,W=1),Translation=(X=1.5,Y=1.5,Z=1),Scale3D=(X=1.1,Y=1,Z=1))))
    Mode=OverrideLocal
    Weight=@n27.Result
    ItemToModify.0.Transform.Rotation=@n33.Result
   )
   > &RigUnit_ModifyTransforms(
    ItemToModify=((Item=(Type=Bone,Name="deform_knee_fr_01_l"),Transform=(Rotation=(X=0,Y=0,Z=0,W=1),Translation=(X=3,Y=1,Z=0),Scale3D=(X=1.1,Y=1,Z=0.75))))
    Mode=OverrideLocal
    Weight=@n30.Result
    ItemToModify.0.Transform.Rotation=@n34.Result
   )
   &RigUnit_ModifyTransforms(
    ItemToModify=((Item=(Type=Bone,Name="deform_knee_bk_01_r"),Transform=(Rotation=(X=0,Y=0,Z=0,W=1),Translation=(X=3,Y=6,Z=0),Scale3D=(X=1,Y=1,Z=0.75))))
    Mode=OverrideLocal
    Weight=@n36.Result
    ItemToModify.0.Transform.Rotation=@n38.Result
    ExecutePin=@n44.A
   )
   > &RigUnit_ModifyTransforms(
    ItemToModify=((Item=(Type=Bone,Name="deform_knee_bk_01_r"),Transform=(Rotation=(X=0,Y=0,Z=0,W=1),Translation=(X=6,Y=3.5,Z=0),Scale3D=(X=1,Y=1,Z=0.75))))
    Mode=OverrideLocal
    Weight=@n39.Result
    ItemToModify.0.Transform.Rotation=@n40.Result
   )
   > &RigUnit_ModifyTransforms(
    ItemToModify=((Item=(Type=Bone,Name="deform_knee_fr_01_r"),Transform=(Rotation=(X=0,Y=0,Z=0,W=1),Translation=(X=-1.5,Y=-1.5,Z=-1),Scale3D=(X=1.1,Y=1,Z=1))))
    Mode=OverrideLocal
    Weight=@n36.Result
    ItemToModify.0.Transform.Rotation=@n42.Result
   )
   > &RigUnit_ModifyTransforms(
    ItemToModify=((Item=(Type=Bone,Name="deform_knee_fr_01_r"),Transform=(Rotation=(X=0,Y=0,Z=0,W=1),Translation=(X=-3,Y=-1,Z=0),Scale3D=(X=1.1,Y=1,Z=0.75))))
    Mode=OverrideLocal
    Weight=@n39.Result
    ItemToModify.0.Transform.Rotation=@n43.Result
   )
   n45 = &RigVMFunction_MathQuaternionToEuler(
    RotationOrder=ZYX
    Value=&RigUnit_GetRelativeTransformForItem(Child=(Type=Bone,Name="lowerarm_l"), Parent=(Type=Bone,Name="upperarm_l"), bChildInitial=false, bParentInitial=false).RelativeTransform.Rotation
   )
   n46 = &RigVMFunction_MathFloatRemap(SourceMaximum=92, SourceMinimum=28, TargetMinimum=0, bClamp=true, Value=@n45.Result.Z)
   n47 = &RigVMFunction_MathQuaternionToEuler(
    RotationOrder=ZYX
    Value=&RigUnit_GetTransform(Item=(Type=Bone,Name="deform_elbow_bk_l"), Space=LocalSpace, bInitial=true).Transform.Rotation
   )
   &RigUnit_ModifyTransforms(
    ItemToModify=((Item=(Type=Bone,Name="deform_elbow_bk_l"),Transform=(Rotation=(X=0,Y=0,Z=0,W=1),Translation=(X=-3.5,Y=-1.75,Z=0),Scale3D=(X=0.5,Y=1,Z=1))))
    Mode=OverrideLocal
    Weight=@n46.Result
    ItemToModify.0.Transform.Rotation=&RigVMFunction_MathQuaternionFromEuler(Euler=&RigVMFunction_MathVectorAdd(B=(X=0,Y=0,Z=-60), A=@n47.Result) {template="Add::Execute(in A,in B,out Result)"}.Result).Result
    ExecutePin=@n16.B
   )
   n48 = &RigVMFunction_MathQuaternionToEuler(
    RotationOrder=ZYX
    Value=&RigUnit_GetTransform(Item=(Type=Bone,Name="deform_elbow_fr_l"), Space=LocalSpace, bInitial=true).Transform.Rotation
   )
   > &RigUnit_ModifyTransforms(
    ItemToModify=((Item=(Type=Bone,Name="deform_elbow_fr_l"),Transform=(Rotation=(X=0,Y=0,Z=0,W=1),Translation=(X=21,Y=4,Z=0),Scale3D=(X=0.5,Y=1,Z=1))))
    Mode=OverrideLocal
    Weight=@n46.Result
    ItemToModify.0.Transform.Rotation=&RigVMFunction_MathQuaternionFromEuler(Euler=&RigVMFunction_MathVectorAdd(B=(X=0,Y=0,Z=10), A=@n48.Result) {template="Add::Execute(in A,in B,out Result)"}.Result).Result
   )
   n49 = &RigVMFunction_MathFloatRemap(SourceMaximum=148, SourceMinimum=91, TargetMinimum=0, bClamp=true, Value=@n45.Result.Z)
   > &RigUnit_ModifyTransforms(
    ItemToModify=((Item=(Type=Bone,Name="deform_elbow_bk_l"),Transform=(Rotation=(X=0,Y=0,Z=0,W=1),Translation=(X=-3.5,Y=0,Z=3),Scale3D=(X=0.7,Y=1,Z=0.6))))
    Mode=OverrideLocal
    Weight=@n49.Result
    ItemToModify.0.Transform.Rotation=&RigVMFunction_MathQuaternionFromEuler(Euler=&RigVMFunction_MathVectorAdd(B=(X=0,Y=0,Z=-90), A=@n47.Result) {template="Add::Execute(in A,in B,out Result)"}.Result).Result
   )
   > &RigUnit_ModifyTransforms(
    ItemToModify=((Item=(Type=Bone,Name="deform_elbow_fr_l"),Transform=(Rotation=(X=0,Y=0,Z=0,W=1),Translation=(X=12,Y=3,Z=-1),Scale3D=(X=0.6,Y=1,Z=1.3))))
    Mode=OverrideLocal
    Weight=@n49.Result
    ItemToModify.0.Transform.Rotation=&RigVMFunction_MathQuaternionFromEuler(Euler=&RigVMFunction_MathVectorAdd(B=(X=0,Y=0,Z=-10), A=@n48.Result) {template="Add::Execute(in A,in B,out Result)"}.Result).Result
   )
   n50 = &RigVMFunction_MathQuaternionToEuler(
    RotationOrder=ZYX
    Value=&RigUnit_GetRelativeTransformForItem(Child=(Type=Bone,Name="lowerarm_r"), Parent=(Type=Bone,Name="upperarm_r"), bChildInitial=false, bParentInitial=false).RelativeTransform.Rotation
   )
   n51 = &RigVMFunction_MathFloatRemap(SourceMaximum=92, SourceMinimum=28, TargetMinimum=0, bClamp=true, Value=@n50.Result.Z)
   n52 = &RigVMFunction_MathQuaternionToEuler(
    RotationOrder=ZYX
    Value=&RigUnit_GetTransform(Item=(Type=Bone,Name="deform_elbow_bk_r"), Space=LocalSpace, bInitial=true).Transform.Rotation
   )
   &RigUnit_ModifyTransforms(
    ItemToModify=((Item=(Type=Bone,Name="deform_elbow_bk_r"),Transform=(Rotation=(X=0,Y=0,Z=0,W=1),Translation=(X=3.5,Y=1.75,Z=0),Scale3D=(X=0.5,Y=1,Z=1))))
    Mode=OverrideLocal
    Weight=@n51.Result
    ItemToModify.0.Transform.Rotation=&RigVMFunction_MathQuaternionFromEuler(Euler=&RigVMFunction_MathVectorAdd(B=(X=0,Y=0,Z=-60), A=@n52.Result) {template="Add::Execute(in A,in B,out Result)"}.Result).Result
    ExecutePin=@n16.A
   )
   n53 = &RigVMFunction_MathQuaternionToEuler(
    RotationOrder=ZYX
    Value=&RigUnit_GetTransform(Item=(Type=Bone,Name="deform_elbow_fr_r"), Space=LocalSpace, bInitial=true).Transform.Rotation
   )
   > &RigUnit_ModifyTransforms(
    ItemToModify=((Item=(Type=Bone,Name="deform_elbow_fr_r"),Transform=(Rotation=(X=0,Y=0,Z=0,W=1),Translation=(X=-21,Y=-4,Z=0),Scale3D=(X=0.5,Y=1,Z=1))))
    Mode=OverrideLocal
    Weight=@n51.Result
    ItemToModify.0.Transform.Rotation=&RigVMFunction_MathQuaternionFromEuler(Euler=&RigVMFunction_MathVectorAdd(B=(X=0,Y=0,Z=-13.6), A=@n53.Result) {template="Add::Execute(in A,in B,out Result)"}.Result).Result
   )
   n54 = &RigVMFunction_MathFloatRemap(SourceMaximum=148, SourceMinimum=91, TargetMinimum=0, bClamp=true, Value=@n50.Result.Z)
   > &RigUnit_ModifyTransforms(
    ItemToModify=((Item=(Type=Bone,Name="deform_elbow_bk_r"),Transform=(Rotation=(X=0,Y=0,Z=0,W=1),Translation=(X=3.5,Y=0,Z=-3),Scale3D=(X=0.7,Y=1,Z=0.6))))
    Mode=OverrideLocal
    Weight=@n54.Result
    ItemToModify.0.Transform.Rotation=&RigVMFunction_MathQuaternionFromEuler(Euler=&RigVMFunction_MathVectorAdd(B=(X=0,Y=0,Z=-90), A=@n52.Result) {template="Add::Execute(in A,in B,out Result)"}.Result).Result
   )
   > &RigUnit_ModifyTransforms(
    ItemToModify=((Item=(Type=Bone,Name="deform_elbow_fr_r"),Transform=(Rotation=(X=0,Y=0,Z=0,W=1),Translation=(X=-12,Y=-3,Z=1),Scale3D=(X=0.6,Y=1,Z=1.3))))
    Mode=OverrideLocal
    Weight=@n54.Result
    ItemToModify.0.Transform.Rotation=&RigVMFunction_MathQuaternionFromEuler(Euler=&RigVMFunction_MathVectorAdd(B=(X=0,Y=0,Z=28), A=@n53.Result) {template="Add::Execute(in A,in B,out Result)"}.Result).Result
   )
   n55 = &RigUnit_SequenceExecution(ExecuteContext=())
   n56 = /Script/ControlRig.RigUnit_CollectionItems(Items=((Type=Bone,Name="pelvis"),(Type=Bone,Name="spine_01"),(Type=Bone,Name="spine_02"),(Type=Bone,Name="spine_03"),(Type=Bone,Name="spine_04"),(Type=Bone,Name="spine_05"))) {is=struct RigVMUnitNode}
   n57 = /Script/ControlRig.RigUnit_CollectionLoop(Collection=@n56.Collection, ExecutePin=@n55.A) {is=struct RigVMUnitNode}
   n58 = /Script/RigVM.RigVMFunction_NameConcat(B=_fk_ctrl, A=@n57.Item.Name) {is=struct RigVMUnitNode, template="Concat::Execute(in A,in B,out Result)"}
   n59 = /Script/ControlRig.RigUnit_ProjectTransformToNewParent(
    Child=(Type=Bone,Name="pelvis")
    NewParent=(Type=Control,Name="pelvis_fk_ctrl")
    OldParent=(Type=Control,Name="pelvis_fk_ctrl")
    bChildInitial=True
    bOldParentInitial=True
    Child.Name=@n57.Item.Name
    NewParent.Name=@n58.Result
    OldParent.Name=@n58.Result
   ) {is=struct RigVMUnitNode}
   > &RigUnit_SetTransform(Item=(Type=Bone,Name="pelvis"), Weight=1, Item.Name=@n57.Item.Name, Value=@n59.Transform)
   n60 = &RigUnit_CollectionItems(Items=((Type=Bone,Name="clavicle_l"),(Type=Bone,Name="upperarm_l"),(Type=Bone,Name="lowerarm_l"),(Type=Bone,Name="hand_l"),(Type=Bone,Name="clavicle_r"),(Type=Bone,Name="upperarm_r"),(Type=Bone,Name="lowerarm_r"),(Type=Bone,Name="hand_r"),()))
   n61 = &RigUnit_CollectionLoop(Collection=@n60.Collection, ExecutePin=@n55.B)
   n62 = &RigVMFunction_NameConcat(B=_fk_ctrl, A=@n61.Item.Name) {template="Concat::Execute(in A,in B,out Result)"}
   > &RigUnit_SetTransform(
    Item=(Type=Bone)
    Weight=1
    Item.Name=@n61.Item.Name
    Value=&RigUnit_ProjectTransformToNewParent(
     Child=(Type=Bone)
     NewParent=(Type=Control)
     OldParent=(Type=Control)
     bChildInitial=True
     bOldParentInitial=True
     Child.Name=@n61.Item.Name
     NewParent.Name=@n62.Result
     OldParent.Name=@n62.Result
    ).Transform
   )
   n63 = &RigUnit_CollectionItems(Items=((Type=Bone,Name="thigh_l"),(Type=Bone,Name="calf_l"),(Type=Bone,Name="foot_l"),(Type=Bone,Name="thigh_r"),(Type=Bone,Name="calf_r"),(Type=Bone,Name="foot_r")))
   n64 = &RigUnit_CollectionLoop(Collection=@n63.Collection, ExecutePin=@n55.C)
   n65 = &RigVMFunction_NameConcat(B=_fk_ctrl, A=@n64.Item.Name) {template="Concat::Execute(in A,in B,out Result)"}
   > &RigUnit_SetTransform(
    Item=(Type=Bone)
    Weight=1
    Item.Name=@n64.Item.Name
    Value=&RigUnit_ProjectTransformToNewParent(
     Child=(Type=Bone)
     NewParent=(Type=Control)
     OldParent=(Type=Control)
     bChildInitial=True
     bOldParentInitial=True
     Child.Name=@n64.Item.Name
     NewParent.Name=@n65.Result
     OldParent.Name=@n65.Result
    ).Transform
   )
   /Script/ControlRig.RigUnit_DebugHierarchy(
    Color=(R=1,G=1,B=1,A=1)
    DebugDrawSettings=(DepthPriority=SDPG_Foreground,Lifetime=-1)
    ExecutePin=()
    Items=()
    Scale=3
    Thickness=0.1
    WorldOffset=(Rotation=(X=0,Y=0,Z=0,W=1),Translation=(X=0,Y=0,Z=0),Scale3D=(X=1,Y=1,Z=1))
    bEnabled=True
   ) {is=struct RigVMUnitNode}
   n66 = &RigVMFunction_MathQuaternionToEuler(RotationOrder=ZYX, Value=&RigUnit_GetInitialBoneTransform(Bone=deform_thigh_bk_l, Space=LocalSpace).Transform.Rotation)
   &RigUnit_ModifyBoneTransforms(
    BoneToModify=((Bone="deform_thigh_bk_l",Transform=(Rotation=(X=0,Y=0,Z=0,W=1),Translation=(X=29.958,Y=-1.401,Z=0.029),Scale3D=(X=1,Y=1,Z=1.803))))
    ExecutePin=()
    Mode=OverrideLocal
    WeightMaximum=1
    WeightMinimum=0
    Weight=@n23.Result
    BoneToModify.0.Transform.Rotation=&RigVMFunction_MathQuaternionFromEuler(Euler=&RigVMFunction_MathVectorAdd(B=(X=0,Y=-37.217,Z=0), A=@n66.Result) {template="Add::Execute(in A,in B,out Result)"}.Result).Result
   )
   n67 = &RigVMFunction_MathQuaternionToEuler(RotationOrder=ZYX, Value=&RigUnit_GetInitialBoneTransform(Bone=deform_thigh_bk_r, Space=LocalSpace).Transform.Rotation)
   &RigUnit_ModifyBoneTransforms(
    BoneToModify=((Bone="deform_thigh_bk_r",Transform=(Rotation=(X=0,Y=0,Z=0,W=1),Translation=(X=-29.958,Y=1.401,Z=-0.029),Scale3D=(X=1,Y=1,Z=1.803))))
    ExecutePin=()
    Mode=OverrideLocal
    WeightMaximum=1
    WeightMinimum=0
    Weight=@n20.Result
    BoneToModify.0.Transform.Rotation=&RigVMFunction_MathQuaternionFromEuler(Euler=&RigVMFunction_MathVectorAdd(B=(X=0,Y=-37.217,Z=0), A=@n67.Result) {template="Add::Execute(in A,in B,out Result)"}.Result).Result
   )
  ]}
]
hierarchy [
 root {loc=<0 0 0>, rot=<0 0 0 1>}
 pelvis {parent=root, loc=<0 -0.7071 93.1191>, rot=<0.0127 0.707 -0.0127 -0.707>}
 spine_01 {parent=pelvis, loc=<6.7494 -0.6388 0>, rot=<0 0 0.034 -0.9994>}
 spine_02 {parent=spine_01, loc=<7.2518 -0.035 0>, rot=<0 0 0.0238 -0.9997>}
 spine_03 {parent=spine_02, loc=<6.0104 0 0>, rot=<0 0 0.0149 -0.9999>}
 spine_04 {parent=spine_03, loc=<8.1024 0 0>, rot=<0 0 -0.0132 -0.9999>}
 spine_05 {parent=spine_04, loc=<10.0146 0 0>, rot=<0 0 -0.0015 -1>}
 clavicle_l {parent=spine_05, loc=<4.3575 -0.1466 -2.3223>, rot=<0.0465 -0.7289 -0.009 -0.6829>}
 upperarm_l {parent=clavicle_l, loc=<15.5111 0 0>, rot=<0.0226 -0.3234 -0.0088 -0.9459>}
 lowerarm_l {parent=upperarm_l, loc=<26.6514 0 0>, rot=<0.0391 -0.0099 -0.2453 -0.9686>}
 hand_l {parent=lowerarm_l, loc=<22.3491 0 0>, rot=<-0.7081 -0.0507 0.0975 -0.6975>}
 index_metacarpal_l {parent=hand_l, loc=<1.6535 0.0183 -2.96>, rot=<0.0158 -0.0124 -0.0173 -0.9996>}
 index_01_l {parent=index_metacarpal_l, loc=<8.3008 0.0357 -0.0524>, rot=<-0.0044 0.0485 0.0924 -0.9945>}
 index_02_l {parent=index_01_l, loc=<3.3587 -0.3004 0.0452>, rot=<-0.002 0.0027 0.3225 -0.9466>}
 index_03_l {parent=index_02_l, loc=<2.5578 -0.123 -0.0252>, rot=<-0.002 0.0067 0.2558 -0.9667>}
 middle_metacarpal_l {parent=hand_l, loc=<2.1856 -0.1725 -0.5984>, rot=<0.006 0.0168 -0.0329 -0.9993>}
 middle_01_l {parent=middle_metacarpal_l, loc=<7.3275 -0.2447 0.0129>, rot=<-0.0208 0.0042 0.0858 -0.9961>}
 middle_02_l {parent=middle_01_l, loc=<4.0123 0.1401 0.1175>, rot=<-0.0218 0.0258 0.3512 -0.9357>}
 middle_03_l {parent=middle_02_l, loc=<2.6077 0.0534 0.1373>, rot=<0.0082 0.0188 0.1708 -0.9851>}
 pinky_metacarpal_l {parent=hand_l, loc=<1.7616 -0.9753 3.3814>, rot=<-0.0397 0.0814 -0.069 -0.9935>}
 pinky_01_l {parent=pinky_metacarpal_l, loc=<7.1709 -0.4831 -0.1094>, rot=<-0.0654 -0.0035 0.2242 -0.9723>}
 pinky_02_l {parent=pinky_01_l, loc=<2.8274 0.0893 -0.1043>, rot=<0.0037 0.0011 0.2854 -0.9584>}
 pinky_03_l {parent=pinky_02_l, loc=<2.2046 0.0307 -0.0078>, rot=<-0.0002 0.0098 0.1445 -0.9895>}
 ring_metacarpal_l {parent=hand_l, loc=<2.2074 -0.3625 1.3638>, rot=<-0.0284 0.0722 -0.0458 -0.9959>}
 ring_01_l {parent=ring_metacarpal_l, loc=<7.0295 -0.6197 -0.1446>, rot=<-0.0489 -0.0252 0.1892 -0.9804>}
 ring_02_l {parent=ring_01_l, loc=<3.6174 0.4647 -0.1027>, rot=<0 0 0.2806 -0.9598>}
 ring_03_l {parent=ring_02_l, loc=<2.4152 0.114 0.0199>, rot=<0 0 0.1189 -0.9929>}
 thumb_01_l {parent=hand_l, loc=<1.3118 -0.3191 -2.459>, rot=<0.6546 -0.4042 0.0664 -0.6354>}
 thumb_02_l {parent=thumb_01_l, loc=<5.9336 -0.0448 0.0401>, rot=<-0.0037 0.0213 0.298 -0.9543>}
 thumb_03_l {parent=thumb_02_l, loc=<2.2223 0.3682 -0.0907>, rot=<0.0118 -0.013 0.0066 -0.9998>}
 weapon_l {parent=hand_l, loc=<-0.1974 -3.4111 0.3811>, rot=<0 0 0 1>}
 lowerarm_twist_01_l {parent=lowerarm_l, loc=<17.8549 0 0>, rot=<0 0 0 1>}
 lowerarm_twist_02_l {parent=lowerarm_l, loc=<9.6199 0 0>, rot=<0 0 0 1>}
 deform_elbow_bk_l {parent=lowerarm_l, loc=<-1.2853 -2.9274 -0.5607>, rot=<-0.0285 0.0335 -0.0024 -0.999>}
 upperarm_twist_01_l {parent=upperarm_l, loc=<0.0041 0 0>, rot=<0 0 0 1>}
 deform_deltoid_fr_l {parent=upperarm_twist_01_l, loc=<2.4511 0 0>, rot=<0 0 0 1>}
 upperarm_twist_02_l {parent=upperarm_l, loc=<13.7941 0 0>, rot=<0 0 0 1>}
 deform_elbow_fr_l {parent=upperarm_l, loc=<26.5048 0.1875 -0.5816>, rot=<0.0021 0.0156 -0.185 -0.9826>}
 deform_pectoral_fr_l {parent=clavicle_l, loc=<9.3905 0 0>, rot=<0 0 0 1>}
 deform_trap_l {parent=clavicle_l, loc=<2.4327 -3.9764 6.2458>, rot=<0 -0.1691 0 -0.9856>}
 clavicle_r {parent=spine_05, loc=<4.3575 -0.1466 2.3223>, rot=<0.7289 0.0465 -0.6829 0.009>}
 upperarm_r {parent=clavicle_r, loc=<-15.5111 0 0>, rot=<0.0226 -0.3234 -0.0088 -0.9459>}
 lowerarm_r {parent=upperarm_r, loc=<-26.6515 0 0>, rot=<0.0391 -0.0099 -0.2453 -0.9686>}
 hand_r {parent=lowerarm_r, loc=<-22.3492 0 0>, rot=<-0.7081 -0.0507 0.0975 -0.6975>}
 index_metacarpal_r {parent=hand_r, loc=<-1.6534 -0.0181 2.9601>, rot=<0.0158 -0.0124 -0.0173 -0.9996>}
 index_01_r {parent=index_metacarpal_r, loc=<-8.3008 -0.0357 0.0524>, rot=<-0.0044 0.0485 0.0924 -0.9945>}
 index_02_r {parent=index_01_r, loc=<-3.3589 0.3003 -0.0453>, rot=<-0.002 0.0027 0.3225 -0.9466>}
 index_03_r {parent=index_02_r, loc=<-2.5578 0.1231 0.0251>, rot=<-0.002 0.0067 0.2558 -0.9667>}
 middle_metacarpal_r {parent=hand_r, loc=<-2.1856 0.1726 0.5984>, rot=<0.006 0.0168 -0.0329 -0.9993>}
 middle_01_r {parent=middle_metacarpal_r, loc=<-7.3273 0.2447 -0.0129>, rot=<-0.0208 0.0042 0.0858 -0.9961>}
 middle_02_r {parent=middle_01_r, loc=<-4.0142 -0.1411 -0.07>, rot=<-0.0217 0.0259 0.3512 -0.9357>}
 middle_03_r {parent=middle_02_r, loc=<-2.6074 -0.0534 -0.1312>, rot=<0.0083 0.0188 0.1708 -0.9851>}
 pinky_metacarpal_r {parent=hand_r, loc=<-1.7617 0.9753 -3.3814>, rot=<-0.0397 0.0814 -0.069 -0.9935>}
 pinky_01_r {parent=pinky_metacarpal_r, loc=<-7.1707 0.4832 0.1094>, rot=<-0.0654 -0.0035 0.2242 -0.9723>}
 pinky_02_r {parent=pinky_01_r, loc=<-2.8274 -0.0892 0.1043>, rot=<0.0037 0.0011 0.2854 -0.9584>}
 pinky_03_r {parent=pinky_02_r, loc=<-2.2045 -0.031 0.0078>, rot=<-0.0002 0.0098 0.1445 -0.9895>}
 ring_metacarpal_r {parent=hand_r, loc=<-2.2073 0.3627 -1.3638>, rot=<-0.0284 0.0722 -0.0458 -0.9959>}
 ring_01_r {parent=ring_metacarpal_r, loc=<-7.0297 0.6197 0.1446>, rot=<-0.0489 -0.0252 0.1892 -0.9804>}
 ring_02_r {parent=ring_01_r, loc=<-3.6171 -0.4648 0.1027>, rot=<0 0 0.2806 -0.9598>}
 ring_03_r {parent=ring_02_r, loc=<-2.4153 -0.1141 -0.0199>, rot=<0 0 0.1189 -0.9929>}
 thumb_01_r {parent=hand_r, loc=<-1.3119 0.3192 2.459>, rot=<0.6546 -0.4042 0.0664 -0.6354>}
 thumb_02_r {parent=thumb_01_r, loc=<-5.9335 0.0447 -0.0401>, rot=<-0.0037 0.0213 0.298 -0.9543>}
 thumb_03_r {parent=thumb_02_r, loc=<-2.2223 -0.3682 0.0907>, rot=<0.0118 -0.013 0.0066 -0.9998>}
 weapon_r {parent=hand_r, loc=<0.1975 3.4112 -0.3811>, rot=<0 0 0 -1>}
 lowerarm_twist_01_r {parent=lowerarm_r, loc=<-17.855 0 0>, rot=<0 0 0 1>}
 lowerarm_twist_02_r {parent=lowerarm_r, loc=<-9.6199 0 0>, rot=<0 0 0 1>}
 deform_elbow_bk_r {parent=lowerarm_r, loc=<1.2852 2.9274 0.561>, rot=<-0.0285 0.0335 -0.0024 -0.999>}
 upperarm_twist_01_r {parent=upperarm_r, loc=<-0.0041 0 0>, rot=<0 0 0 1>}
 deform_deltoid_fr_r {parent=upperarm_twist_01_r, loc=<-2.4512 0 0>, rot=<0 0 0 1>}
 upperarm_twist_02_r {parent=upperarm_r, loc=<-13.7942 0 0>, rot=<0 0 0 1>}
 deform_elbow_fr_r {parent=upperarm_r, loc=<-26.5046 -0.1875 0.5813>, rot=<0.0021 0.0156 -0.185 -0.9826>}
 deform_pectoral_fr_r {parent=clavicle_r, loc=<-9.39 0 0>, rot=<0 0 0 1>}
 deform_trap_r {parent=clavicle_r, loc=<-2.4327 3.9764 -6.2455>, rot=<0 -0.1691 0 -0.9856>}
 neck_01 {parent=spine_05, loc=<11.3138 0.1529 0>, rot=<0 0 -0.2003 -0.9797>}
 neck_02 {parent=neck_01, loc=<4.5627 0 0>, rot=<0 0 0.0504 -0.9987>}
 head {parent=neck_02, loc=<5.3914 0 0>, rot=<0 0 0.115 -0.9934>}
 thigh_l {parent=pelvis, loc=<-2.5693 -0.0109 -7.7501>, rot=<0.0012 0.0572 0.0155 -0.9982>}
 calf_l {parent=thigh_l, loc=<-39.0484 0 0>, rot=<0.0006 -0.0161 -0.0391 -0.9991>}
 calf_twist_01_l {parent=calf_l, loc=<-35.813 -0.0967 0.0397>, rot=<0.0237 0.0095 -0.0035 -0.9997>}
 calf_twist_02_l {parent=calf_l, loc=<-18.8064 -0.0967 0.0397>, rot=<0.0237 0.0095 -0.0035 -0.9997>}
 foot_l {parent=calf_l, loc=<-42.561 0.1077 0.4292>, rot=<0.0418 -0.0455 0.0343 -0.9975>}
 ball_l {parent=foot_l, loc=<-6.1599 13.8264 0.6806>, rot=<0.0001 -0.0001 -0.7063 -0.7079>}
 deform_knee_bk_01_l {parent=calf_l, loc=<0.6022 -1.0607 0>, rot=<0 0 -0.0408 -0.9992>}
 deform_knee_fr_01_l {parent=calf_l, loc=<-0.2342 0 0>, rot=<0 0 0 1>}
 thigh_twist_01_l {parent=thigh_l, loc=<-21.6303 0 0>, rot=<0.0013 0 0 -1>}
 deform_thigh_bk_l {parent=thigh_twist_01_l, loc=<16.9514 0.2058 0.1808>, rot=<-0.0013 -0.0217 0 -0.9998>}
 deform_thigh_fr_l {parent=thigh_twist_01_l, loc=<21.5084 0.2281 -0.0598>, rot=<-0.0013 -0.0217 0 -0.9998>}
 deform_glut_l {parent=thigh_l, loc=<0 0 0>, rot=<0 0 0 1>}
 thigh_r {parent=pelvis, loc=<-2.5691 -0.0109 7.7501>, rot=<0.0572 -0.0012 0.9982 0.0155>}
 calf_r {parent=thigh_r, loc=<39.0487 0 0>, rot=<0.0006 -0.0161 -0.0391 -0.9991>}
 calf_twist_01_r {parent=calf_r, loc=<35.8132 0.0967 -0.0397>, rot=<0.0237 0.0095 -0.0035 -0.9997>}
 calf_twist_02_r {parent=calf_r, loc=<18.8064 0.0967 -0.0397>, rot=<0.0237 0.0095 -0.0035 -0.9997>}
 foot_r {parent=calf_r, loc=<42.5612 -0.1077 -0.4292>, rot=<0.0418 -0.0455 0.0343 -0.9975>}
 ball_r {parent=foot_r, loc=<6.1598 -13.8264 -0.6806>, rot=<0.0001 -0.0001 -0.7063 -0.7079>}
 deform_knee_bk_01_r {parent=calf_r, loc=<-0.6022 1.0607 0>, rot=<0 0 -0.0408 -0.9992>}
 deform_knee_fr_01_r {parent=calf_r, loc=<0.2341 0 0>, rot=<0 0 0 1>}
 thigh_twist_01_r {parent=thigh_r, loc=<21.6303 0 0>, rot=<0.0013 0 0 -1>}
 deform_thigh_bk_r {parent=thigh_twist_01_r, loc=<-16.9512 -0.2058 -0.1808>, rot=<-0.0013 -0.0217 0 -0.9998>}
 deform_thigh_fr_r {parent=thigh_twist_01_r, loc=<-21.5083 -0.2281 0.0598>, rot=<-0.0013 -0.0217 0 -0.9998>}
 deform_glut_r {parent=thigh_r, loc=<0 0 0>, rot=<0 0 0 1>}
 deform_pelvis {parent=pelvis, loc=<0 0 0>, rot=<0 0 0 1>}
 attach {parent=root, loc=<0 0 0>, rot=<0 0 0 1>}
 ik_foot_root {parent=root, loc=<0 0 0>, rot=<0 0 0 1>}
 ik_foot_l {parent=ik_foot_root, loc=<15.2608 -4.4176 9.4433>, rot=<0.0357 0.7045 0.0284 -0.7082>}
 ik_foot_r {parent=ik_foot_root, loc=<-15.2609 -4.4176 9.4431>, rot=<0.7082 0.0284 -0.7045 0.0357>}
 ik_hand_root {parent=root, loc=<0 0 0>, rot=<0 0 0 1>}
 ik_hand_gun {parent=ik_hand_root, loc=<-51.9729 4.0884 102.6935>, rot=<0.6084 0.1793 0.4613 -0.6205>}
 ik_hand_l {parent=ik_hand_gun, loc=<53.0286 82.1782 35.2051>, rot=<-0.5102 -0.7906 -0.3387 0>}
 ik_hand_r {parent=ik_hand_gun, loc=<0 0 0>, rot=<0 0 0 1>}
 pelvis_fk_ctrl {controlType=EULER_TRANSFORM, type=Control, loc=<0 0 0>, rot=<0 0 0 1>}
 thigh_l_fk_ctrl {controlType=EULER_TRANSFORM, parent=pelvis_fk_ctrl, type=Control, loc=<-96.4759 1.265 21.4743>, rot=<-0.0106 0.7071 -0.011 0.707>}
 calf_l_fk_ctrl {controlType=EULER_TRANSFORM, parent=thigh_l_fk_ctrl, type=Control, loc=<-47.1336 4.7644 55.3432>, rot=<0.0601 0.744 0.0004 0.6655>}
 foot_l_fk_ctrl {controlType=EULER_TRANSFORM, parent=calf_l_fk_ctrl, type=Control, loc=<-9.2972 -0.6301 57.951>, rot=<-0.0194 0.735 0.0351 0.6769>}
 thigh_r_fk_ctrl {controlType=EULER_TRANSFORM, parent=pelvis_fk_ctrl, type=Control, loc=<81.6643 -1.194 -14.6>, rot=<0.0077 -0.707 -0.0141 0.707>}
 calf_r_fk_ctrl {controlType=EULER_TRANSFORM, parent=thigh_r_fk_ctrl, type=Control, loc=<0 0 0>, rot=<0 0 0 1>}
 foot_r_fk_ctrl {controlType=EULER_TRANSFORM, parent=calf_r_fk_ctrl, type=Control, loc=<0 0 0>, rot=<0 0 0 1>}
 spine_01_fk_ctrl {controlType=EULER_TRANSFORM, parent=pelvis_fk_ctrl, type=Control, loc=<-99.85 -2.7564 -6.7493>, rot=<-0.0355 0.7062 -0.0127 0.707>}
 spine_02_fk_ctrl {controlType=EULER_TRANSFORM, parent=spine_01_fk_ctrl, type=Control, loc=<0 0 0>, rot=<0 0 0 1>}
 spine_03_fk_ctrl {controlType=EULER_TRANSFORM, parent=spine_02_fk_ctrl, type=Control, loc=<0 0 0>, rot=<0 0 0 1>}
 spine_04_fk_ctrl {controlType=EULER_TRANSFORM, parent=spine_03_fk_ctrl, type=Control, loc=<0 0 0>, rot=<0 0 0 1>}
 spine_05_fk_ctrl {controlType=EULER_TRANSFORM, parent=spine_04_fk_ctrl, type=Control, loc=<0 0 0>, rot=<0 0 0 1>}
 clavicle_r_fk_ctrl {controlType=EULER_TRANSFORM, parent=spine_05_fk_ctrl, type=Control, loc=<-2.0804 -3.5767 133.3293>, rot=<0.0092 -0.7071 -0.0278 0.7065>}
 upperarm_r_fk_ctrl {controlType=EULER_TRANSFORM, parent=clavicle_r_fk_ctrl, type=Control, loc=<0 0 0>, rot=<0 0 0 1>}
 lowerarm_r_fk_ctrl {controlType=EULER_TRANSFORM, parent=upperarm_r_fk_ctrl, type=Control, loc=<0 0 0>, rot=<0 0 0 1>}
 hand_r_fk_ctrl {controlType=EULER_TRANSFORM, parent=lowerarm_r_fk_ctrl, type=Control, loc=<0 0 0>, rot=<0 0 0 1>}
 clavicle_l_fk_ctrl {controlType=EULER_TRANSFORM, parent=spine_05_fk_ctrl, type=Control, loc=<11.0763 4.0157 -137.3795>, rot=<-0.0657 0.7042 0.024 0.7065>}
 upperarm_l_fk_ctrl {controlType=EULER_TRANSFORM, parent=clavicle_l_fk_ctrl, type=Control, loc=<87.4059 9.4933 -101.8604>, rot=<-0.0162 -0.0344 0.0177 0.9991>}
 lowerarm_l_fk_ctrl {controlType=EULER_TRANSFORM, parent=upperarm_l_fk_ctrl, type=Control, loc=<68.9615 -17.6999 -93.3689>, rot=<-0.1574 -0.319 -0.0147 0.9345>}
 hand_l_fk_ctrl {controlType=EULER_TRANSFORM, parent=lowerarm_l_fk_ctrl, type=Control, loc=<0 0 0>, rot=<0 0 0 1>}
]
variables [
 actor_emote {type=bool, default=False}
 emote_glut_weight {type=double, default=1}
 emote_thigh_weight {type=double, default=0.5}
 glut_twist_arms_crossed_l {type=FVector, default=(X=0,Y=0,Z=0)}
 glut_twist_arms_crossed_r {type=FVector, default=(X=0,Y=0,Z=0)}
 is_arms_crossed {type=bool, default=False}
 thigh_twist_arms_crossed_l {type=FVector, default=(X=0,Y=0,Z=0)}
 thigh_twist_arms_crossed_r {type=FVector, default=(X=0,Y=0,Z=0)}
]
