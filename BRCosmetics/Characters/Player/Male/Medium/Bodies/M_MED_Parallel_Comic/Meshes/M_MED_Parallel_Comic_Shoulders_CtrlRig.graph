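# Control Rig for the male Parallel Comic body that procedurally corrects shoulder deformation through helper bones.
# Upperarm twist bones counter-rotate by a fraction of each upper arm's twist, extracted by swing-twist decomposition.
# Arm swing weights pectoral and deltoid offsets, clavicle pitch weights trapezius offsets, and `is_arms_crossed` swaps the right side to a fixed deltoid pose.
# FK control-to-bone loops exist but are disconnected from execution.
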ControlRigBlueprint /BRCosmetics/Characters/Player/Male/Medium/Bodies/M_MED_Parallel_Comic/Meshes/M_MED_Parallel_Comic_Shoulders_CtrlRig
graphs [
 RigVMFunctionLibrary {links [], nodes []}
 RigVMModel {nodes [
   RigUnit_BeginExecution = &RigUnit_BeginExecution()
   n1 = /Script/ControlRig.RigUnit_SequenceExecution(ExecuteContext=@RigUnit_BeginExecution.ExecutePin) {is=struct RigVMUnitNode}
   n2 = &RigUnit_SequenceExecution(ExecuteContext=@n1.B)
   n3 = /Script/RigVM.RigVMFunction_MathQuaternionToAxisAndAngle(Value=&RigVMFunction_MathQuaternionSwingTwist(TwistAxis=(X=1,Y=0,Z=0), Input=&RigUnit_GetBoneTransform(Bone=upperarm_l, Space=LocalSpace).Transform.Rotation).Twist) {is=struct RigVMUnitNode, template="ToAxisAndAngle::Execute(in Value,out Axis,out Angle)"}
   RerouteNode_9 = RigVMDispatchNode(Value=-0.1) {template="DISPATCH_RigVMDispatch_Constant(io Value)"}
   /Script/ControlRig.RigUnit_SetBoneRotation(
    Bone=upperarm_twist_01_l
    Weight=1
    bPropagateToChildren=True
    Rotation=/Script/RigVM.RigVMFunction_MathQuaternionFromAxisAndAngle(
     Angle=&RigVMFunction_MathFloatMul(A=@n3.Angle, B=@RerouteNode_9.Value) {template="Multiply::Execute(in A,in B,out Result)"}.Result
     Axis=@n3.Axis
    ) {is=struct RigVMUnitNode, template="FromAxisAndAngle::Execute(in Axis,in Angle,out Result)"}.Result
    ExecutePin=@n2.A
   ) {is=struct RigVMUnitNode}
   n4 = &RigVMFunction_MathQuaternionToAxisAndAngle(Value=&RigVMFunction_MathQuaternionSwingTwist(TwistAxis=(X=1,Y=0,Z=0), Input=&RigUnit_GetBoneTransform(Bone=upperarm_l, Space=LocalSpace).Transform.Rotation).Twist) {template="ToAxisAndAngle::Execute(in Value,out Axis,out Angle)"}
   RerouteNode = RigVMDispatchNode(Value=-0.1) {template="DISPATCH_RigVMDispatch_Constant(io Value)"}
   > &RigUnit_SetBoneRotation(
    Bone=upperarm_twist_02_l
    Weight=1
    bPropagateToChildren=True
    Rotation=&RigVMFunction_MathQuaternionFromAxisAndAngle(
     Angle=&RigVMFunction_MathFloatMul(A=@n4.Angle, B=@RerouteNode.Value) {template="Multiply::Execute(in A,in B,out Result)"}.Result
     Axis=@n4.Axis
    ) {template="FromAxisAndAngle::Execute(in Axis,in Angle,out Result)"}.Result
   )
   n5 = &RigVMFunction_MathQuaternionToAxisAndAngle(Value=&RigVMFunction_MathQuaternionSwingTwist(TwistAxis=(X=1,Y=0,Z=0), Input=&RigUnit_GetBoneTransform(Bone=upperarm_r, Space=LocalSpace).Transform.Rotation).Twist) {template="ToAxisAndAngle::Execute(in Value,out Axis,out Angle)"}
   > &RigUnit_SetBoneRotation(
    Bone=upperarm_twist_01_r
    Weight=1
    bPropagateToChildren=True
    Rotation=&RigVMFunction_MathQuaternionFromAxisAndAngle(
     Angle=&RigVMFunction_MathFloatMul(A=@n5.Angle, B=@RerouteNode_9.Value) {template="Multiply::Execute(in A,in B,out Result)"}.Result
     Axis=@n5.Axis
    ) {template="FromAxisAndAngle::Execute(in Axis,in Angle,out Result)"}.Result
   )
   n6 = &RigVMFunction_MathQuaternionToAxisAndAngle(Value=&RigVMFunction_MathQuaternionSwingTwist(TwistAxis=(X=1,Y=0,Z=0), Input=&RigUnit_GetBoneTransform(Bone=upperarm_r, Space=LocalSpace).Transform.Rotation).Twist) {template="ToAxisAndAngle::Execute(in Value,out Axis,out Angle)"}
   > &RigUnit_SetBoneRotation(
    Bone=upperarm_twist_02_r
    Weight=1
    bPropagateToChildren=True
    Rotation=&RigVMFunction_MathQuaternionFromAxisAndAngle(
     Angle=&RigVMFunction_MathFloatMul(A=@n6.Angle, B=@RerouteNode.Value) {template="Multiply::Execute(in A,in B,out Result)"}.Result
     Axis=@n6.Axis
    ) {template="FromAxisAndAngle::Execute(in Axis,in Angle,out Result)"}.Result
   )
   n7 = &RigUnit_SequenceExecution(ExecuteContext=@n2.C)
   Branch_RigVMUnitNode = &RigVMFunction_ControlFlowBranch(ExecuteContext=@n7.A, Condition=RigVMVariableNode() {variable=is_arms_crossed}.Value)
   n8 = &RigVMFunction_MathQuaternionToEuler(
    RotationOrder=ZYX
    Value=&RigUnit_GetTransform(Item=(Type=Bone,Name="deform_deltoid_fr_r"), Space=LocalSpace, bInitial=true).Transform.Rotation
   )
   &RigUnit_ModifyTransforms(
    ItemToModify=((Item=(Type=Bone,Name="deform_deltoid_fr_r"),Transform=(Rotation=(X=0,Y=0,Z=0,W=1),Translation=(X=-5,Y=0,Z=0),Scale3D=(X=1,Y=1.2,Z=1))))
    Mode=OverrideLocal
    Weight=1
    ItemToModify.0.Transform.Rotation=&RigVMFunction_MathQuaternionFromEuler(Euler=/Script/RigVM.RigVMFunction_MathVectorAdd(B=(X=60,Y=0,Z=-30), A=@n8.Result) {is=struct RigVMUnitNode, template="Add::Execute(in A,in B,out Result)"}.Result).Result
    ExecutePin=@Branch_RigVMUnitNode.True
   )
   n9 = &RigVMFunction_MathTransformMakeRelative(
    Parent=&RigUnit_GetTransform(Item=(Type=Bone,Name="upperarm_r"), Space=LocalSpace, bInitial=true).Transform
    Global=&RigUnit_GetTransform(Item=(Type=Bone,Name="upperarm_r"), Space=LocalSpace).Transform
   )
   n10 = &RigVMFunction_MathFloatRemap(
    SourceMaximum=120
    SourceMinimum=0
    TargetMinimum=0
    bClamp=true
    Value=&RigVMFunction_MathQuaternionToEuler(RotationOrder=ZYX, Value=&RigVMFunction_MathQuaternionSwingTwist(TwistAxis=(X=1,Y=0,Z=0), Input=@n9.Local.Rotation).Swing).Result.Z
   )
   n11 = &RigVMFunction_MathQuaternionToEuler(
    RotationOrder=ZYX
    Value=&RigUnit_GetTransform(Item=(Type=Bone,Name="deform_pectoral_fr_r"), Space=LocalSpace, bInitial=true).Transform.Rotation
   )
   &RigUnit_ModifyTransforms(
    ItemToModify=((Item=(Type=Bone,Name="deform_pectoral_fr_r"),Transform=(Rotation=(X=0,Y=0,Z=0,W=1),Translation=(X=-6.5,Y=-6,Z=-2),Scale3D=(X=0.7,Y=1,Z=1))))
    Mode=OverrideLocal
    Weight=@n10.Result
    ItemToModify.0.Transform.Rotation=&RigVMFunction_MathQuaternionFromEuler(Euler=&RigVMFunction_MathVectorAdd(B=(X=10,Y=0,Z=25), A=@n11.Result) {template="Add::Execute(in A,in B,out Result)"}.Result).Result
    ExecutePin=@Branch_RigVMUnitNode.False
   )
   > &RigUnit_ModifyTransforms(
    ItemToModify=((Item=(Type=Bone,Name="deform_deltoid_fr_r"),Transform=(Rotation=(X=0,Y=0,Z=0,W=1),Translation=(X=-6,Y=-2.5,Z=0),Scale3D=(X=0.9,Y=0.9,Z=0.9))))
    Mode=OverrideLocal
    Weight=@n10.Result
   )
   n12 = &RigUnit_SequenceExecution(ExecuteContext=@n7.C)
   n13 = &RigUnit_GetTransform(Item=(Type=Bone,Name="clavicle_l"), Space=LocalSpace, bInitial=false)
   n14 = &RigUnit_GetTransform(Item=(Type=Bone,Name="clavicle_l"), Space=LocalSpace, bInitial=true)
   n15 = &RigVMFunction_MathQuaternionToEuler(
    RotationOrder=ZYX
    Value=&RigVMFunction_MathQuaternionSwingTwist(
     TwistAxis=(X=1,Y=0,Z=0)
     Input=&RigVMFunction_MathTransformMakeRelative(Global=@n13.Transform, Parent=@n14.Transform).Local.Rotation
    ).Swing
   )
   n16 = &RigVMFunction_MathQuaternionToEuler(
    RotationOrder=ZYX
    Value=&RigUnit_GetTransform(Item=(Type=Bone,Name="deform_trap_l"), Space=LocalSpace, bInitial=true).Transform.Rotation
   )
   &RigUnit_ModifyTransforms(
    ItemToModify=((Item=(Type=Bone,Name="deform_trap_l"),Transform=(Rotation=(X=0,Y=0,Z=0,W=1),Translation=(X=4,Y=-4,Z=11),Scale3D=(X=0.6,Y=1,Z=1.2))))
    Mode=OverrideLocal
    Weight=&RigVMFunction_MathFloatRemap(SourceMaximum=-90, SourceMinimum=-5, bClamp=true, Value=@n15.Result.Y).Result
    ItemToModify.0.Transform.Rotation=&RigVMFunction_MathQuaternionFromEuler(Euler=&RigVMFunction_MathVectorAdd(B=(X=0,Y=65,Z=10), A=@n16.Result) {template="Add::Execute(in A,in B,out Result)"}.Result).Result
    ExecutePin=@n12.B
   )
   n17 = &RigVMFunction_MathQuaternionToEuler(
    RotationOrder=ZYX
    Value=&RigUnit_GetTransform(Item=(Type=Bone,Name="deform_trap_l"), Space=LocalSpace, bInitial=true).Transform.Rotation
   )
   > &RigUnit_ModifyTransforms(
    ItemToModify=((Item=(Type=Bone,Name="deform_trap_l"),Transform=(Rotation=(X=0,Y=0,Z=0,W=1),Translation=(X=5,Y=-4,Z=4),Scale3D=(X=1,Y=1,Z=1))))
    Mode=OverrideLocal
    Weight=&RigVMFunction_MathFloatRemap(SourceMaximum=90, SourceMinimum=5, bClamp=true, Value=@n15.Result.Y).Result
    ItemToModify.0.Transform.Rotation=&RigVMFunction_MathQuaternionFromEuler(Euler=&RigVMFunction_MathVectorAdd(B=(X=0,Y=-10,Z=10), A=@n17.Result) {template="Add::Execute(in A,in B,out Result)"}.Result).Result
   )
   n18 = &RigVMFunction_MathQuaternionToEuler(
    RotationOrder=ZYX
    Value=&RigVMFunction_MathTransformMakeRelative(
     Global=&RigUnit_GetTransform(Item=(Type=Bone,Name="clavicle_r"), Space=LocalSpace, bInitial=false).Transform
     Parent=&RigUnit_GetTransform(Item=(Type=Bone,Name="clavicle_r"), Space=LocalSpace, bInitial=true).Transform
    ).Local.Rotation
   )
   n19 = &RigVMFunction_MathQuaternionToEuler(
    RotationOrder=ZYX
    Value=&RigUnit_GetTransform(Item=(Type=Bone,Name="deform_trap_r"), Space=LocalSpace, bInitial=true).Transform.Rotation
   )
   &RigUnit_ModifyTransforms(
    ItemToModify=((Item=(Type=Bone,Name="deform_trap_r"),Transform=(Rotation=(X=0,Y=0,Z=0,W=1),Translation=(X=-4,Y=4,Z=-11),Scale3D=(X=0.6,Y=1,Z=1.2))))
    Mode=OverrideLocal
    Weight=&RigVMFunction_MathFloatRemap(SourceMaximum=-90, SourceMinimum=-5, bClamp=true, Value=@n18.Result.Y).Result
    ItemToModify.0.Transform.Rotation=&RigVMFunction_MathQuaternionFromEuler(Euler=&RigVMFunction_MathVectorAdd(B=(X=0,Y=65,Z=10), A=@n19.Result) {template="Add::Execute(in A,in B,out Result)"}.Result).Result
    ExecutePin=@n12.A
   )
   n20 = &RigVMFunction_MathQuaternionToEuler(
    RotationOrder=ZYX
    Value=&RigUnit_GetTransform(Item=(Type=Bone,Name="deform_trap_r"), Space=LocalSpace, bInitial=true).Transform.Rotation
   )
   > &RigUnit_ModifyTransforms(
    ItemToModify=((Item=(Type=Bone,Name="deform_trap_r"),Transform=(Rotation=(X=0,Y=0,Z=0,W=1),Translation=(X=-5,Y=4,Z=-5),Scale3D=(X=1,Y=1,Z=1))))
    Mode=OverrideLocal
    Weight=&RigVMFunction_MathFloatRemap(SourceMaximum=90, SourceMinimum=5, Value=@n18.Result.Y).Result
    ItemToModify.0.Transform.Rotation=&RigVMFunction_MathQuaternionFromEuler(Euler=&RigVMFunction_MathVectorAdd(B=(X=0,Y=-10,Z=10), A=@n20.Result) {template="Add::Execute(in A,in B,out Result)"}.Result).Result
   )
   n21 = &RigVMFunction_MathTransformMakeRelative(
    Parent=&RigUnit_GetTransform(Item=(Type=Bone,Name="upperarm_l"), Space=LocalSpace, bInitial=true).Transform
    Global=&RigUnit_GetTransform(Item=(Type=Bone,Name="upperarm_l"), Space=LocalSpace).Transform
   )
   n22 = &RigVMFunction_MathFloatRemap(
    SourceMaximum=120
    SourceMinimum=0
    TargetMinimum=0
    bClamp=true
    Value=&RigVMFunction_MathQuaternionToEuler(RotationOrder=ZYX, Value=&RigVMFunction_MathQuaternionSwingTwist(TwistAxis=(X=1,Y=0,Z=0), Input=@n21.Local.Rotation).Swing).Result.Z
   )
   n23 = &RigVMFunction_MathQuaternionToEuler(
    RotationOrder=ZYX
    Value=&RigUnit_GetTransform(Item=(Type=Bone,Name="deform_pectoral_fr_l"), Space=LocalSpace, bInitial=true).Transform.Rotation
   )
   &RigUnit_ModifyTransforms(
    ItemToModify=((Item=(Type=Bone,Name="deform_pectoral_fr_l"),Transform=(Rotation=(X=0,Y=0,Z=0,W=1),Translation=(X=6.5,Y=6,Z=2),Scale3D=(X=0.7,Y=1,Z=1))))
    Mode=OverrideLocal
    Weight=@n22.Result
    ItemToModify.0.Transform.Rotation=&RigVMFunction_MathQuaternionFromEuler(Euler=&RigVMFunction_MathVectorAdd(B=(X=10,Y=0,Z=25), A=@n23.Result) {template="Add::Execute(in A,in B,out Result)"}.Result).Result
    ExecutePin=@n7.B
   )
   n24 = &RigVMFunction_MathQuaternionToEuler(
    RotationOrder=ZYX
    Value=&RigUnit_GetTransform(Item=(Type=Bone,Name="deform_deltoid_fr_l"), Space=LocalSpace, bInitial=true).Transform.Rotation
   )
   > &RigUnit_ModifyTransforms(
    ItemToModify=((Item=(Type=Bone,Name="deform_deltoid_fr_l"),Transform=(Rotation=(X=0,Y=0,Z=0,W=1),Translation=(X=6,Y=2.5,Z=0),Scale3D=(X=0.9,Y=0.9,Z=0.9))))
    Mode=OverrideLocal
    Weight=@n22.Result
    ItemToModify.0.Transform.Rotation=&RigVMFunction_MathQuaternionFromEuler(Euler=&RigVMFunction_MathVectorAdd(B=(X=15,Y=-35,Z=-35), A=@n24.Result) {template="Add::Execute(in A,in B,out Result)"}.Result).Result
   )
   n25 = &RigUnit_SequenceExecution(ExecuteContext=())
   n26 = /Script/ControlRig.RigUnit_CollectionItems(Items=((Type=Bone,Name="pelvis"),(Type=Bone,Name="spine_01"),(Type=Bone,Name="spine_02"),(Type=Bone,Name="spine_03"),(Type=Bone,Name="spine_04"),(Type=Bone,Name="spine_05"))) {is=struct RigVMUnitNode}
   n27 = /Script/ControlRig.RigUnit_CollectionLoop(Collection=@n26.Collection, ExecutePin=@n25.A) {is=struct RigVMUnitNode}
   n28 = /Script/RigVM.RigVMFunction_NameConcat(B=_fk_ctrl, A=@n27.Item.Name) {is=struct RigVMUnitNode, template="Concat::Execute(in A,in B,out Result)"}
   n29 = /Script/ControlRig.RigUnit_ProjectTransformToNewParent(
    Child=(Type=Bone,Name="pelvis")
    NewParent=(Type=Control,Name="pelvis_fk_ctrl")
    OldParent=(Type=Control,Name="pelvis_fk_ctrl")
    bChildInitial=True
    bOldParentInitial=True
    Child.Name=@n27.Item.Name
    NewParent.Name=@n28.Result
    OldParent.Name=@n28.Result
   ) {is=struct RigVMUnitNode}
   > &RigUnit_SetTransform(Item=(Type=Bone,Name="pelvis"), Weight=1, Item.Name=@n27.Item.Name, Value=@n29.Transform)
   n30 = &RigUnit_CollectionItems(Items=((Type=Bone,Name="clavicle_l"),(Type=Bone,Name="upperarm_l"),(Type=Bone,Name="lowerarm_l"),(Type=Bone,Name="hand_l"),(Type=Bone,Name="clavicle_r"),(Type=Bone,Name="upperarm_r"),(Type=Bone,Name="lowerarm_r"),(Type=Bone,Name="hand_r"),()))
   n31 = &RigUnit_CollectionLoop(Collection=@n30.Collection, ExecutePin=@n25.B)
   n32 = &RigVMFunction_NameConcat(B=_fk_ctrl, A=@n31.Item.Name) {template="Concat::Execute(in A,in B,out Result)"}
   > &RigUnit_SetTransform(
    Item=(Type=Bone)
    Weight=1
    Item.Name=@n31.Item.Name
    Value=&RigUnit_ProjectTransformToNewParent(
     Child=(Type=Bone)
     NewParent=(Type=Control)
     OldParent=(Type=Control)
     bChildInitial=True
     bOldParentInitial=True
     Child.Name=@n31.Item.Name
     NewParent.Name=@n32.Result
     OldParent.Name=@n32.Result
    ).Transform
   )
   n33 = &RigUnit_CollectionItems(Items=((Type=Bone,Name="thigh_l"),(Type=Bone,Name="calf_l"),(Type=Bone,Name="foot_l"),(Type=Bone,Name="thigh_r"),(Type=Bone,Name="calf_r"),(Type=Bone,Name="foot_r")))
   n34 = &RigUnit_CollectionLoop(Collection=@n33.Collection, ExecutePin=@n25.C)
   n35 = &RigVMFunction_NameConcat(B=_fk_ctrl, A=@n34.Item.Name) {template="Concat::Execute(in A,in B,out Result)"}
   > &RigUnit_SetTransform(
    Item=(Type=Bone)
    Weight=1
    Item.Name=@n34.Item.Name
    Value=&RigUnit_ProjectTransformToNewParent(
     Child=(Type=Bone)
     NewParent=(Type=Control)
     OldParent=(Type=Control)
     bChildInitial=True
     bOldParentInitial=True
     Child.Name=@n34.Item.Name
     NewParent.Name=@n35.Result
     OldParent.Name=@n35.Result
    ).Transform
   )
   /Script/ControlRig.RigUnit_DebugHierarchy(
    Color=(R=1,G=1,B=1,A=1)
    DebugDrawSettings=(DepthPriority=SDPG_Foreground,Lifetime=-1)
    ExecutePin=()
    Items=()
    Scale=3
    Thickness=0.1
    WorldOffset=(Rotation=(X=0,Y=0,Z=0,W=1),Translation=(X=0,Y=0,Z=0),Scale3D=(X=1,Y=1,Z=1))
    bEnabled=True
   ) {is=struct RigVMUnitNode}
   n36 = &RigUnit_GetTransform(Item=(Type=Bone,Name="deform_deltoid_fr_r"), Space=LocalSpace, bInitial=true)
   &RigVMFunction_MathQuaternionFromEuler(Euler=&RigVMFunction_MathVectorAdd(B=(X=15,Y=-35,Z=-35), A=&RigVMFunction_MathQuaternionToEuler(RotationOrder=ZYX, Value=@n36.Transform.Rotation).Result) {template="Add::Execute(in A,in B,out Result)"}.Result)
  ]}
]
hierarchy [
 root {loc=<0 0 0>, rot=<0 0 0 1>}
 attach {parent=root, loc=<0 0 0>, rot=<0 0 0 1>}
 pelvis {parent=root, loc=<0 -0.7071 93.1191>, rot=<0.0127 0.707 -0.0127 -0.707>}
 spine_01 {parent=pelvis, loc=<6.7494 -0.6388 0>, rot=<0 0 0.034 -0.9994>}
 spine_02 {parent=spine_01, loc=<7.2518 -0.035 0>, rot=<0 0 0.0238 -0.9997>}
 spine_03 {parent=spine_02, loc=<6.0104 0 0>, rot=<0 0 0.0149 -0.9999>}
 spine_04 {parent=spine_03, loc=<8.1024 0 0>, rot=<0 0 -0.0132 -0.9999>}
 spine_05 {parent=spine_04, loc=<10.0146 0 0>, rot=<0 0 -0.0015 -1>}
 clavicle_l {parent=spine_05, loc=<4.3575 -0.1466 -2.3223>, rot=<0.0465 -0.7289 -0.009 -0.6829>}
 upperarm_l {parent=clavicle_l, loc=<15.5111 0 0>, rot=<0.0226 -0.3234 -0.0088 -0.9459>}
 lowerarm_l {parent=upperarm_l, loc=<26.6514 0 0>, rot=<0.0391 -0.0099 -0.2453 -0.9686>}
 hand_l {parent=lowerarm_l, loc=<22.3491 0 0>, rot=<-0.7081 -0.0507 0.0975 -0.6975>}
 index_metacarpal_l {parent=hand_l, loc=<1.6535 0.0183 -2.96>, rot=<0.0158 -0.0124 -0.0173 -0.9996>}
 index_01_l {parent=index_metacarpal_l, loc=<8.3008 0.0357 -0.0524>, rot=<-0.0044 0.0485 0.0924 -0.9945>}
 index_02_l {parent=index_01_l, loc=<3.3587 -0.3004 0.0452>, rot=<-0.002 0.0027 0.3225 -0.9466>}
 index_03_l {parent=index_02_l, loc=<2.5578 -0.123 -0.0252>, rot=<-0.002 0.0067 0.2558 -0.9667>}
 middle_metacarpal_l {parent=hand_l, loc=<2.1856 -0.1725 -0.5984>, rot=<0.006 0.0168 -0.0329 -0.9993>}
 middle_01_l {parent=middle_metacarpal_l, loc=<7.3275 -0.2447 0.0129>, rot=<-0.0208 0.0042 0.0858 -0.9961>}
 middle_02_l {parent=middle_01_l, loc=<4.0123 0.1401 0.1175>, rot=<-0.0218 0.0258 0.3512 -0.9357>}
 middle_03_l {parent=middle_02_l, loc=<2.6077 0.0534 0.1373>, rot=<0.0082 0.0188 0.1708 -0.9851>}
 pinky_metacarpal_l {parent=hand_l, loc=<1.7616 -0.9753 3.3814>, rot=<-0.0397 0.0814 -0.069 -0.9935>}
 pinky_01_l {parent=pinky_metacarpal_l, loc=<7.1709 -0.4831 -0.1094>, rot=<-0.0654 -0.0035 0.2242 -0.9723>}
 pinky_02_l {parent=pinky_01_l, loc=<2.8274 0.0893 -0.1043>, rot=<0.0037 0.0011 0.2854 -0.9584>}
 pinky_03_l {parent=pinky_02_l, loc=<2.2046 0.0307 -0.0078>, rot=<-0.0002 0.0098 0.1445 -0.9895>}
 ring_metacarpal_l {parent=hand_l, loc=<2.2074 -0.3625 1.3638>, rot=<-0.0284 0.0722 -0.0458 -0.9959>}
 ring_01_l {parent=ring_metacarpal_l, loc=<7.0295 -0.6197 -0.1446>, rot=<-0.0489 -0.0252 0.1892 -0.9804>}
 ring_02_l {parent=ring_01_l, loc=<3.6174 0.4647 -0.1027>, rot=<0 0 0.2806 -0.9598>}
 ring_03_l {parent=ring_02_l, loc=<2.4152 0.114 0.0199>, rot=<0 0 0.1189 -0.9929>}
 thumb_01_l {parent=hand_l, loc=<1.3118 -0.3191 -2.459>, rot=<0.6546 -0.4042 0.0664 -0.6354>}
 thumb_02_l {parent=thumb_01_l, loc=<5.9336 -0.0448 0.0401>, rot=<-0.0037 0.0213 0.298 -0.9543>}
 thumb_03_l {parent=thumb_02_l, loc=<2.2223 0.3682 -0.0907>, rot=<0.0118 -0.013 0.0066 -0.9998>}
 weapon_l {parent=hand_l, loc=<-0.1974 -3.4111 0.3811>, rot=<0 0 0 1>}
 lowerarm_twist_01_l {parent=lowerarm_l, loc=<17.8549 0 0>, rot=<0 0 0 1>}
 lowerarm_twist_02_l {parent=lowerarm_l, loc=<9.6199 0 0>, rot=<0 0 0 1>}
 deform_elbow_bk_l {parent=lowerarm_l, loc=<-1.2853 -2.9274 -0.5607>, rot=<-0.0285 0.0335 -0.0024 -0.999>}
 upperarm_twist_01_l {parent=upperarm_l, loc=<0.0041 0 0>, rot=<0 0 0 1>}
 deform_deltoid_fr_l {parent=upperarm_twist_01_l, loc=<2.4511 0 0>, rot=<0 0 0 1>}
 upperarm_twist_02_l {parent=upperarm_l, loc=<13.7941 0 0>, rot=<0 0 0 1>}
 deform_elbow_fr_l {parent=upperarm_l, loc=<26.5048 0.1875 -0.5816>, rot=<0.0021 0.0156 -0.185 -0.9826>}
 deform_pectoral_fr_l {parent=clavicle_l, loc=<9.3905 0 0>, rot=<0 0 0 1>}
 deform_trap_l {parent=clavicle_l, loc=<2.4327 -3.9764 6.2458>, rot=<0 -0.1691 0 -0.9856>}
 clavicle_r {parent=spine_05, loc=<4.3575 -0.1466 2.3223>, rot=<0.7289 0.0465 -0.6829 0.009>}
 upperarm_r {parent=clavicle_r, loc=<-15.5111 0 0>, rot=<0.0226 -0.3234 -0.0088 -0.9459>}
 lowerarm_r {parent=upperarm_r, loc=<-26.6515 0 0>, rot=<0.0391 -0.0099 -0.2453 -0.9686>}
 hand_r {parent=lowerarm_r, loc=<-22.3492 0 0>, rot=<-0.7081 -0.0507 0.0975 -0.6975>}
 index_metacarpal_r {parent=hand_r, loc=<-1.6534 -0.0181 2.9601>, rot=<0.0158 -0.0124 -0.0173 -0.9996>}
 index_01_r {parent=index_metacarpal_r, loc=<-8.3008 -0.0357 0.0524>, rot=<-0.0044 0.0485 0.0924 -0.9945>}
 index_02_r {parent=index_01_r, loc=<-3.3589 0.3003 -0.0453>, rot=<-0.002 0.0027 0.3225 -0.9466>}
 index_03_r {parent=index_02_r, loc=<-2.5578 0.1231 0.0251>, rot=<-0.002 0.0067 0.2558 -0.9667>}
 middle_metacarpal_r {parent=hand_r, loc=<-2.1856 0.1726 0.5984>, rot=<0.006 0.0168 -0.0329 -0.9993>}
 middle_01_r {parent=middle_metacarpal_r, loc=<-7.3273 0.2447 -0.0129>, rot=<-0.0208 0.0042 0.0858 -0.9961>}
 middle_02_r {parent=middle_01_r, loc=<-4.0142 -0.1411 -0.07>, rot=<-0.0217 0.0259 0.3512 -0.9357>}
 middle_03_r {parent=middle_02_r, loc=<-2.6074 -0.0534 -0.1312>, rot=<0.0083 0.0188 0.1708 -0.9851>}
 pinky_metacarpal_r {parent=hand_r, loc=<-1.7617 0.9753 -3.3814>, rot=<-0.0397 0.0814 -0.069 -0.9935>}
 pinky_01_r {parent=pinky_metacarpal_r, loc=<-7.1707 0.4832 0.1094>, rot=<-0.0654 -0.0035 0.2242 -0.9723>}
 pinky_02_r {parent=pinky_01_r, loc=<-2.8274 -0.0892 0.1043>, rot=<0.0037 0.0011 0.2854 -0.9584>}
 pinky_03_r {parent=pinky_02_r, loc=<-2.2045 -0.031 0.0078>, rot=<-0.0002 0.0098 0.1445 -0.9895>}
 ring_metacarpal_r {parent=hand_r, loc=<-2.2073 0.3627 -1.3638>, rot=<-0.0284 0.0722 -0.0458 -0.9959>}
 ring_01_r {parent=ring_metacarpal_r, loc=<-7.0297 0.6197 0.1446>, rot=<-0.0489 -0.0252 0.1892 -0.9804>}
 ring_02_r {parent=ring_01_r, loc=<-3.6171 -0.4648 0.1027>, rot=<0 0 0.2806 -0.9598>}
 ring_03_r {parent=ring_02_r, loc=<-2.4153 -0.1141 -0.0199>, rot=<0 0 0.1189 -0.9929>}
 thumb_01_r {parent=hand_r, loc=<-1.3119 0.3192 2.459>, rot=<0.6546 -0.4042 0.0664 -0.6354>}
 thumb_02_r {parent=thumb_01_r, loc=<-5.9335 0.0447 -0.0401>, rot=<-0.0037 0.0213 0.298 -0.9543>}
 thumb_03_r {parent=thumb_02_r, loc=<-2.2223 -0.3682 0.0907>, rot=<0.0118 -0.013 0.0066 -0.9998>}
 weapon_r {parent=hand_r, loc=<0.1975 3.4112 -0.3811>, rot=<0 0 0 -1>}
 lowerarm_twist_01_r {parent=lowerarm_r, loc=<-17.855 0 0>, rot=<0 0 0 1>}
 lowerarm_twist_02_r {parent=lowerarm_r, loc=<-9.6199 0 0>, rot=<0 0 0 1>}
 deform_elbow_bk_r {parent=lowerarm_r, loc=<1.2852 2.9274 0.561>, rot=<-0.0285 0.0335 -0.0024 -0.999>}
 upperarm_twist_01_r {parent=upperarm_r, loc=<-0.0041 0 0>, rot=<0 0 0 1>}
 deform_deltoid_fr_r {parent=upperarm_twist_01_r, loc=<-2.4512 0 0>, rot=<0 0 0 1>}
 upperarm_twist_02_r {parent=upperarm_r, loc=<-13.7942 0 0>, rot=<0 0 0 1>}
 deform_elbow_fr_r {parent=upperarm_r, loc=<-26.5046 -0.1875 0.5813>, rot=<0.0021 0.0156 -0.185 -0.9826>}
 deform_pectoral_fr_r {parent=clavicle_r, loc=<-9.39 0 0>, rot=<0 0 0 1>}
 deform_trap_r {parent=clavicle_r, loc=<-2.4327 3.9764 -6.2455>, rot=<0 -0.1691 0 -0.9856>}
 neck_01 {parent=spine_05, loc=<11.3138 0.1529 0>, rot=<0 0 -0.2003 -0.9797>}
 neck_02 {parent=neck_01, loc=<4.5627 0 0>, rot=<0 0 0.0504 -0.9987>}
 head {parent=neck_02, loc=<5.3914 0 0>, rot=<0 0 0.115 -0.9934>}
 thigh_l {parent=pelvis, loc=<-2.5693 -0.0109 -7.7501>, rot=<0.0012 0.0572 0.0155 -0.9982>}
 calf_l {parent=thigh_l, loc=<-39.0484 0 0>, rot=<0.0006 -0.0161 -0.0391 -0.9991>}
 calf_twist_01_l {parent=calf_l, loc=<-35.813 -0.0967 0.0397>, rot=<0.0237 0.0095 -0.0035 -0.9997>}
 calf_twist_02_l {parent=calf_l, loc=<-18.8064 -0.0967 0.0397>, rot=<0.0237 0.0095 -0.0035 -0.9997>}
 foot_l {parent=calf_l, loc=<-42.561 0.1077 0.4292>, rot=<0.0418 -0.0455 0.0343 -0.9975>}
 ball_l {parent=foot_l, loc=<-6.1599 13.8264 0.6806>, rot=<0.0001 -0.0001 -0.7063 -0.7079>}
 deform_knee_bk_01_l {parent=calf_l, loc=<0.6022 -1.0607 0>, rot=<0 0 -0.0408 -0.9992>}
 deform_knee_fr_01_l {parent=calf_l, loc=<-0.2342 0 0>, rot=<0 0 0 1>}
 thigh_twist_01_l {parent=thigh_l, loc=<-21.6303 0 0>, rot=<0.0013 0 0 -1>}
 deform_thigh_bk_l {parent=thigh_twist_01_l, loc=<16.9514 0.2058 0.1808>, rot=<-0.0013 -0.0217 0 -0.9998>}
 deform_thigh_fr_l {parent=thigh_twist_01_l, loc=<21.5084 0.2281 -0.0598>, rot=<-0.0013 -0.0217 0 -0.9998>}
 deform_glut_l {parent=thigh_l, loc=<0 0 0>, rot=<0 0 0 1>}
 thigh_r {parent=pelvis, loc=<-2.5691 -0.0109 7.7501>, rot=<0.0572 -0.0012 0.9982 0.0155>}
 calf_r {parent=thigh_r, loc=<39.0487 0 0>, rot=<0.0006 -0.0161 -0.0391 -0.9991>}
 calf_twist_01_r {parent=calf_r, loc=<35.8132 0.0967 -0.0397>, rot=<0.0237 0.0095 -0.0035 -0.9997>}
 calf_twist_02_r {parent=calf_r, loc=<18.8064 0.0967 -0.0397>, rot=<0.0237 0.0095 -0.0035 -0.9997>}
 foot_r {parent=calf_r, loc=<42.5612 -0.1077 -0.4292>, rot=<0.0418 -0.0455 0.0343 -0.9975>}
 ball_r {parent=foot_r, loc=<6.1598 -13.8264 -0.6806>, rot=<0.0001 -0.0001 -0.7063 -0.7079>}
 deform_knee_bk_01_r {parent=calf_r, loc=<-0.6022 1.0607 0>, rot=<0 0 -0.0408 -0.9992>}
 deform_knee_fr_01_r {parent=calf_r, loc=<0.2341 0 0>, rot=<0 0 0 1>}
 thigh_twist_01_r {parent=thigh_r, loc=<21.6303 0 0>, rot=<0.0013 0 0 -1>}
 deform_thigh_bk_r {parent=thigh_twist_01_r, loc=<-16.9512 -0.2058 -0.1808>, rot=<-0.0013 -0.0217 0 -0.9998>}
 deform_thigh_fr_r {parent=thigh_twist_01_r, loc=<-21.5083 -0.2281 0.0598>, rot=<-0.0013 -0.0217 0 -0.9998>}
 deform_glut_r {parent=thigh_r, loc=<0 0 0>, rot=<0 0 0 1>}
 deform_pelvis {parent=pelvis, loc=<0 0 0>, rot=<0 0 0 1>}
 ik_foot_root {parent=root, loc=<0 0 0>, rot=<0 0 0 1>}
 ik_foot_l {parent=ik_foot_root, loc=<15.2608 -4.4176 9.4433>, rot=<0.0357 0.7045 0.0284 -0.7082>}
 ik_foot_r {parent=ik_foot_root, loc=<-15.2609 -4.4176 9.4431>, rot=<0.7082 0.0284 -0.7045 0.0357>}
 ik_hand_root {parent=root, loc=<0 0 0>, rot=<0 0 0 1>}
 ik_hand_gun {parent=ik_hand_root, loc=<-51.9729 4.0884 102.6935>, rot=<0.6084 0.1793 0.4613 -0.6205>}
 ik_hand_l {parent=ik_hand_gun, loc=<53.0286 82.1782 35.2051>, rot=<-0.5102 -0.7906 -0.3387 0>}
 ik_hand_r {parent=ik_hand_gun, loc=<0 0 0>, rot=<0 0 0 1>}
 pelvis_fk_ctrl {controlType=EULER_TRANSFORM, type=Control, loc=<0 0 0>, rot=<0 0 0 1>}
 thigh_l_fk_ctrl {controlType=EULER_TRANSFORM, parent=pelvis_fk_ctrl, type=Control, loc=<-96.4755 1.265 21.4742>, rot=<-0.0106 0.7071 -0.011 0.707>}
 calf_l_fk_ctrl {controlType=EULER_TRANSFORM, parent=thigh_l_fk_ctrl, type=Control, loc=<-47.1335 4.7644 55.3431>, rot=<0.0601 0.744 0.0004 0.6655>}
 foot_l_fk_ctrl {controlType=EULER_TRANSFORM, parent=calf_l_fk_ctrl, type=Control, loc=<-9.2972 -0.6301 57.9511>, rot=<-0.0194 0.735 0.0351 0.6769>}
 thigh_r_fk_ctrl {controlType=EULER_TRANSFORM, parent=pelvis_fk_ctrl, type=Control, loc=<81.6643 -1.194 -14.6>, rot=<0.0077 -0.707 -0.0141 0.707>}
 calf_r_fk_ctrl {controlType=EULER_TRANSFORM, parent=thigh_r_fk_ctrl, type=Control, loc=<0 0 0>, rot=<0 0 0 1>}
 foot_r_fk_ctrl {controlType=EULER_TRANSFORM, parent=calf_r_fk_ctrl, type=Control, loc=<0 0 0>, rot=<0 0 0 1>}
 spine_01_fk_ctrl {controlType=EULER_TRANSFORM, parent=pelvis_fk_ctrl, type=Control, loc=<-99.8498 -2.7564 -6.7493>, rot=<-0.0355 0.7062 -0.0127 0.707>}
 spine_02_fk_ctrl {controlType=EULER_TRANSFORM, parent=spine_01_fk_ctrl, type=Control, loc=<0 0 0>, rot=<0 0 0 1>}
 spine_03_fk_ctrl {controlType=EULER_TRANSFORM, parent=spine_02_fk_ctrl, type=Control, loc=<0 0 0>, rot=<0 0 0 1>}
 spine_04_fk_ctrl {controlType=EULER_TRANSFORM, parent=spine_03_fk_ctrl, type=Control, loc=<0 0 0>, rot=<0 0 0 1>}
 spine_05_fk_ctrl {controlType=EULER_TRANSFORM, parent=spine_04_fk_ctrl, type=Control, loc=<0 0 0>, rot=<0 0 0 1>}
 clavicle_r_fk_ctrl {controlType=EULER_TRANSFORM, parent=spine_05_fk_ctrl, type=Control, loc=<-2.0802 -3.5767 133.3291>, rot=<0.0092 -0.7071 -0.0278 0.7065>}
 upperarm_r_fk_ctrl {controlType=EULER_TRANSFORM, parent=clavicle_r_fk_ctrl, type=Control, loc=<0 0 0>, rot=<0 0 0 1>}
 lowerarm_r_fk_ctrl {controlType=EULER_TRANSFORM, parent=upperarm_r_fk_ctrl, type=Control, loc=<0 0 0>, rot=<0 0 0 1>}
 hand_r_fk_ctrl {controlType=EULER_TRANSFORM, parent=lowerarm_r_fk_ctrl, type=Control, loc=<0 0 0>, rot=<0 0 0 1>}
 clavicle_l_fk_ctrl {controlType=EULER_TRANSFORM, parent=spine_05_fk_ctrl, type=Control, loc=<11.0763 4.0157 -137.3793>, rot=<-0.0657 0.7042 0.024 0.7065>}
 upperarm_l_fk_ctrl {controlType=EULER_TRANSFORM, parent=clavicle_l_fk_ctrl, type=Control, loc=<87.4059 9.4933 -101.8604>, rot=<-0.0162 -0.0344 0.0177 0.9991>}
 lowerarm_l_fk_ctrl {controlType=EULER_TRANSFORM, parent=upperarm_l_fk_ctrl, type=Control, loc=<68.9614 -17.6999 -93.3688>, rot=<-0.1574 -0.319 -0.0147 0.9345>}
 hand_l_fk_ctrl {controlType=EULER_TRANSFORM, parent=lowerarm_l_fk_ctrl, type=Control, loc=<0 0 0>, rot=<0 0 0 1>}
]
variables [
 is_arms_crossed {type=bool, default=False}
 upper_arm_l_rel_quat {type=FQuat, default=(X=0,Y=0,Z=0,W=1)}
 upper_arm_r_rel_quat {type=FQuat, default=(X=0,Y=0,Z=0,W=1)}
 upper_arm_twists {type=double, default=0}
]
