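# Animation Blueprint for the M_MED_Lateral player body cosmetic, layering corrective Control Rigs over a pose copied from the attached parent mesh.
# Reuses the Parallel Comic body and shoulder Control Rigs in series, both driven by an arms-crossed flag, the first also by an emote flag.
# Each update derives these flags from montages active on the owner mesh's anim instance, setting arms-crossed only in the front end.
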AnimBlueprint /BRCosmetics/Characters/Player/Male/Medium/Bodies/M_MED_Lateral/Meshes/M_MED_Lateral_AnimBP
target_skeleton=&/M_MED_Lateral_Skeleton
graphs [
 AnimGraph {nodes [
   n1 = Get is_arms_crossed() -> (is_arms_crossed:bool)
   n2 = Control Rig!(
    Source=Copy Pose From Mesh() {bUseAttachedParent=true}
    Alpha
    actor_emote:bool=Get is_playing_emote() -> (is_playing_emote:bool)
    is_arms_crossed:bool=@n1
   ) -> (Pose) {
    alphaBoolBlend {blendOption=Linear}
    controlRigAssetReference {blueprintRigClass=/BRCosmetics/Characters/Player/Male/Medium/Bodies/M_MED_Parallel_Comic/Meshes/M_MED_Parallel_Comic_CtrlRig.M_MED_Parallel_Comic_CtrlRig_C}
   }
   n3 = Control Rig!(Source=@n2, Alpha, is_arms_crossed=@n1) -> (Pose) {
    alphaBoolBlend {blendOption=Linear}
    controlRigAssetReference {blueprintRigClass=/BRCosmetics/Characters/Player/Male/Medium/Bodies/M_MED_Parallel_Comic/Meshes/M_MED_Parallel_Comic_Shoulders_CtrlRig.M_MED_Parallel_Comic_Shoulders_CtrlRig_C}
   }
   Output Pose(Result=@n3) {layerGroup=None}
  ]}
 EventGraph {nodes [
   Event BlueprintUpdateAnimation()
   > n1 = Sequence()
   n2 = GetAnimInstance(self=GetOwnerSkeletalMeshComponent())
   > Is Valid(InputObject=@n2)
   > Set is_playing_emote(is_playing_emote:bool=Montage_IsActive!(
     self:Anim Instance Object Reference=@n2
     Montage:Anim Montage Object Reference=/BRCosmetics/Animation/Game/MainPlayer/Emotes/ParallelComic/ParallelComic_CMM_M
    ) -> (ReturnValue:bool)) -> (Output_Get:bool)
   Branch(execute=@n1.then_1, Condition=Get bIsInFrontEnd())
   > Set is_arms_crossed(is_arms_crossed:bool=Montage_IsActive(self=@n2, Montage=/BRCosmetics/Animation/Game/MainPlayer/Menu/BR/CustomSkins/_s19/ParallelComic/ParallelComic_Male_Idle_Montage)) -> (Output_Get:bool)
  ]}
]
variables [
 is_arms_crossed {type=boolean}
 is_playing_emote {type=boolean}
]
