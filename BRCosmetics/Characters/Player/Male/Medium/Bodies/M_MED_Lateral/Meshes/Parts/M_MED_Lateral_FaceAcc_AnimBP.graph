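# Animation Blueprint for the face accessory part of the M_MED_Lateral male player cosmetic, driving it from other character meshes.
# It copies the attached parent's pose and layers a second copy, sourced through body_mesh, over clavicle, pectoral, deltoid and upper-arm twist bones.
# The result feeds a shoulder Control Rig borrowed from Parallel_Comic, whose arms-crossed flag tracks that skin's idle montage, checked behind a front-end branch.
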AnimBlueprint /BRCosmetics/Characters/Player/Male/Medium/Bodies/M_MED_Lateral/Meshes/Parts/M_MED_Lateral_FaceAcc_AnimBP
target_skeleton=&/M_MED_Lateral_FaceAcc_Skeleton
graphs [
 AnimGraph {nodes [
   n1 = Layered blend per bone(
    BasePose=Copy Pose From Mesh() {bUseAttachedParent=true}
    BlendPoses_0=Copy Pose From Mesh(SourceMeshComponent=Get body_mesh() -> (body_mesh:Skeletal Mesh Component Object Reference)) {bUseAttachedParent=true}
   ) {layerSetup [{branchFilters [
       {boneName=clavicle_l}
       {boneName=clavicle_r}
       {boneName=deform_pectoral_fr_l}
       {boneName=deform_pectoral_fr_r}
       {boneName=upperarm_twist_01_l}
       {boneName=upperarm_twist_01_r}
       {boneName=deform_deltoid_fr_l}
       {boneName=deform_deltoid_fr_r}
       {boneName=upperarm_twist_02_l}
       {boneName=upperarm_twist_02_r}
      ]}]}
   n2 = Control Rig!(Source=@n1, Alpha, is_arms_crossed:bool=Get is_arms_crossed() -> (is_arms_crossed:bool)) -> (Pose) {
    alphaBoolBlend {blendOption=Linear}
    controlRigAssetReference {blueprintRigClass=/BRCosmetics/Characters/Player/Male/Medium/Bodies/M_MED_Parallel_Comic/Meshes/M_MED_Parallel_Comic_Shoulders_CtrlRig.M_MED_Parallel_Comic_Shoulders_CtrlRig_C}
   }
   Output Pose(Result=@n2) {layerGroup=None}
  ]}
 EventGraph {nodes [
   Event BlueprintUpdateAnimation()
   > Branch(Condition=Get bIsInFrontEnd())
   n1 = GetOwnerSkeletalMeshComponent()
   > Set is_arms_crossed(is_arms_crossed:bool=Montage_IsActive!(
     self:Anim Instance Object Reference=GetAnimInstance(self=@n1)
     Montage:Anim Montage Object Reference=/BRCosmetics/Animation/Game/MainPlayer/Menu/BR/CustomSkins/_s19/ParallelComic/ParallelComic_Male_Idle_Montage
    ) -> (ReturnValue:bool)) -> (Output_Get:bool)
   Event BlueprintInitializeAnimation()
   > Set body_mesh(body_mesh:Skeletal Mesh Component Object Reference=GetOwnerSkeletalMeshComponent()) -> (Output_Get:Skeletal Mesh Component Object Reference)
   > n2 = Cast To PlayerPawn_Athena(Object=TryGetPawnOwner())
   > n3 = Set pawn_owner(pawn_owner:Fort Player Pawn Athena Object Reference=@n2) -> (Output_Get:Fort Player Pawn Athena Object Reference)
   > Set head_mesh(head_mesh:Skeletal Mesh Component Object Reference=GetSkeletalMeshForPartType(self=@n3, PartType=Head)) -> (Output_Get:Skeletal Mesh Component Object Reference)
  ]}
]
variables [
 body_mesh {type=SkeletalMeshComponent}
 head_mesh {type=SkeletalMeshComponent}
 is_arms_crossed {type=boolean}
 pawn_owner {type=FortPlayerPawnAthena}
]
